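# Material function that layers sequins with animated sparkle onto incoming material attributes within a Mask, for the ElegantLilyAnkle character body.
# Tiled sequin textures supply offset, roughness and normals, with a switch to reuse base roughness and normals, and ShirtMask separating shirt from shoe normal intensity.
# A Fresnel input drives two-color falloff and brightness, while time-scrolled noise times panning caustics creates emissive sparkle.
# Static switches gate each output attribute.
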MaterialFunction /BRCosmetics/Characters/Player/Female/Medium/Bodies/F_MED_ElegantLilyAnkle/Materials/MF_ElegantLilyAnkleSequins
graph {nodes [
  n1 = FunctionInput(Preview=SetMaterialAttributes!(MaterialAttributes)) {bUsePreviewValueAsDefault=true, inputName=inMat, inputType=FunctionInput_MaterialAttributes}
  n2 = GetMaterialAttributes(None=@n1) {attributeGetTypes [#1, #2]}
  n3 = GetMaterialAttributes(None=@n1) {attributeGetTypes [
    #3
    #4
    #1
    #5
    #2
   ]}
  n4 = FunctionInput() {inputName=Mask, inputType=FunctionInput_Scalar}
  n5 = FunctionInput() {bUsePreviewValueAsDefault=true, inputName=Fresnel, inputType=FunctionInput_Scalar}
  n6 = OneMinus(None=@n5)
  n7 = Multiply(A=@n6, B=@n5)
  n8 = Multiply(
   A=ScalarParameter() {defaultValue=1.5, parameterName=SequinBrightness}
   B=Multiply(A=NamedRerouteUsage(), B=LinearInterpolate(Alpha=@n7) {constA=0.7})
  )
  n9 = Constant()
  n10 = ScalarParameter() {defaultValue=0.9, parameterName=SequinMetalness}
  n11 = TextureSample(UVs=NamedRerouteUsage()) {
   samplerSource=SSM_Wrap_WorldGroupSettings
   samplerType=SAMPLERTYPE_LinearGrayscale
   texture=/Game/Global/Textures/Default/Tileable_Patterns/T_SequinTile_roughness
  }
  n12 = StaticSwitchParameter(True=NamedRerouteUsage(), False=@n11.RGB) {parameterName=SequinsUseBaseNormalRoughness}
  n13 = Time()
  n14 = ScalarParameter() {defaultValue=0.01, parameterName=SequinSparkleSpeed}
  n15 = TextureSample(UVs=Add(A=NamedRerouteUsage(), B=Multiply(A=@n13, B=@n14) {constB=0.2})) {samplerSource=SSM_Wrap_WorldGroupSettings, samplerType=SAMPLERTYPE_Normal, texture=/Game/Global/Textures/Default/Noise/64x64Noise}
  n16 = Panner(Coordinate=Multiply(A=TextureCoordinate(), B=ScalarParameter() {defaultValue=2, parameterName=SequinDiamondTile})) {speedX=0.05, speedY=0.05}
  n17 = Multiply(
   A=Power(Base=@n15.R) {constExponent=4}
   B=Power(Base=TextureSample(UVs=@n16) {texture=/Game/Effects/Fort_Effects/Textures/Tile/T_ShieldCaustics_Tile_Chromatic}.R)
  )
  n18 = VectorParameter() {parameterName=SparkleColor, defaultValue=<rgba 0.5742 0.228 1 1>}
  n19 = Multiply(A=NamedRerouteUsage(), B=Multiply(A=Multiply(A=@n17, B=@n18.RGB), B=ScalarParameter() {defaultValue=10, parameterName=SparkleBRightness}))
  n20 = NamedRerouteUsage()
  n21 = Multiply(
   A=LinearInterpolate(
    A=ScalarParameter() {defaultValue=0.8, parameterName=SequinNormalIntensity_Shoes}
    B=ScalarParameter() {parameterName=SequinNormalIntensity_Shirt}
    Alpha=NamedRerouteUsage()
   )
   B=Constant3Vector() {constant=<rgba 1 1 0 0>}
  )
  n22 = TextureSample(UVs=NamedRerouteUsage()) {
   samplerSource=SSM_Wrap_WorldGroupSettings
   samplerType=SAMPLERTYPE_Normal
   texture=/Game/Global/Textures/Default/Tileable_Patterns/T_SequinTile_N
  }
  n23 = StaticSwitchParameter(True=Multiply(A=@n20, B=@n21), False=Add(A=@n20, B=Multiply(A=@n22.RGB, B=@n21)) {constB=0}) {parameterName=SequinsUseBaseNormalRoughness}
  n24 = SetMaterialAttributes!(
   MaterialAttributes=@n2.MaterialAttributes
   BaseColor=StaticSwitchParameter(True=LinearInterpolate(A=@n3.BaseColor, B=@n8, Alpha=@n4), False=@n9) {defaultValue=true, group=09_DisableTextureSamplers, parameterName=MFSequin_UseBaseColor}
   Metallic=StaticSwitchParameter(True=LinearInterpolate(A=@n3.Metallic, B=@n10, Alpha=@n4), False=@n9) {defaultValue=true, group=09_DisableTextureSamplers, parameterName=MFSequin_UseMetallic}
   Roughness=StaticSwitchParameter(True=LinearInterpolate(A=@n3.Roughness, B=@n12, Alpha=@n4), False=@n9) {defaultValue=true, group=09_DisableTextureSamplers, parameterName=MFSequin_UseRoughness}
   EmissiveColor=StaticSwitchParameter(True=LinearInterpolate(A=@n3.EmissiveColor, B=@n19, Alpha=@n4), False=@n9) {defaultValue=true, group=09_DisableTextureSamplers, parameterName=MFSequin_UseEmissiveColor}
   Normal=StaticSwitchParameter(True=LinearInterpolate(A=@n3.Normal, B=@n23, Alpha=@n4), False=@n9) {defaultValue=true, group=09_DisableTextureSamplers, parameterName=MFSequin_UseNormal}
  ) {attributeSetTypes [
    #3
    #4
    #1
    #5
    #2
   ]}
  FunctionOutput(None=@n24)
  NamedRerouteDeclaration(None=Multiply(A=TextureCoordinate(), B=ScalarParameter() {defaultValue=32, parameterName=SequinTile})) {name=SequinUVs, nodeColor=<rgba 0 1 0.0859 1>}
  n25 = TextureSampleParameter2D(UVs=NamedRerouteUsage()) {parameterName=SequinOffset, samplerType=SAMPLERTYPE_LinearGrayscale, texture=/Game/Global/Textures/Default/Tileable_Patterns/T_SequinTile}
  NamedRerouteDeclaration(None=@n25.R) {name=SequinOffset, nodeColor=<rgba 1 0 0.6562 1>}
  n26 = VectorParameter() {parameterName=SequinFalloffColor01, defaultValue=<rgba 0.8842 0.6282 1 1>}
  n27 = VectorParameter() {parameterName=SequinFalloffColor02, defaultValue=<rgba 0.6663 0.575 1 1>}
  n28 = LinearInterpolate(Alpha=NamedRerouteUsage()) {constA=0.5, constB=1.5}
  NamedRerouteDeclaration(None=LinearInterpolate(
    A=FunctionInput(Preview=@n26.RGB) {bUsePreviewValueAsDefault=true, inputName=SequinColorOuter, sortPriority=2}
    B=FunctionInput(Preview=@n27.RGB) {bUsePreviewValueAsDefault=true, inputName=SequinColorInner, sortPriority=3}
    Alpha=Multiply(A=@n28, B=@n6)
   )) {name=SequinColor, nodeColor=<rgba 0.8984 0 1 1>}
  NamedRerouteDeclaration(None=@n2.Roughness) {name=BaseRoughness, nodeColor=<rgba 0.6172 1 0 1>}
  NamedRerouteDeclaration(None=@n2.Normal) {name=BaseNormals, nodeColor=<rgba 0 1 0.1562 1>}
  NamedRerouteDeclaration(None=FunctionInput() {inputName=ShirtMask, inputType=FunctionInput_Scalar, sortPriority=100}) {name=ShirtSequins, nodeColor=<rgba 0.1016 0 1 1>}
 ]}
