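# Parent character material for the ElegantLilyAnkle female body cosmetic, adding a sequin effect to the outfit.
# It wraps a sequin function with the shared 2023 character PreFX and PostFX functions, using material attributes.
# The sequin function takes a mask from the FX texture's green channel, Fresnel, and outer colors lerped from base color toward purple.
# A shirt mask comes from vertex-interpolated pre-skinned local height, so it holds through animation.
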Material /BRCosmetics/Characters/Player/Female/Medium/Bodies/F_MED_ElegantLilyAnkle/Materials/M_ElegantLilyAnkle_Parent
graph {
 properties {
  bUseMaterialAttributes=true
  bUsedWithMeshDeformer=true
  bUsedWithMorphTargets=true
  bUsedWithSkeletalMesh=true
  pixelDepthOffsetMode=PDOM_Legacy
  shadingModel=MSM_FromMaterialExpression
  usedShadingModels=Default Lit | Subsurface | Clear Coat | Subsurface Profile | Cloth
 }
 nodes [
  n1 = MaterialFunctionCall() {materialFunction=&MF_CharacterParent_2023_PreFX}
  MaterialFunctionCall(Material=NamedRerouteUsage(), M=@n1.M, Pre Skinned Local Position=@n1.M, Distance Blend=@n1.M, Fresnel=@n1.M) {materialFunction=&MF_CharacterParent_2023_PostFX}
  NamedRerouteDeclaration(None=@n1.Result) {name=OriginalMA, nodeColor=<rgba 1 0 0.1406 1>}
  n2 = GetMaterialAttributes(None=NamedRerouteUsage()) {attributeGetTypes [#1, #2]}
  n3 = Constant3Vector() {constant=<rgba 0.1635 0 0.2969 1>}
  n4 = NamedRerouteUsage()
  NamedRerouteDeclaration(None=MaterialFunctionCall!(
    Mask=NamedRerouteUsage()
    Fresnel=NamedRerouteUsage()
    inMat=@n2.MaterialAttributes
    SequinColorOuter=LinearInterpolate(A=@n2.Base Color, B=@n3, Alpha=@n4)
    SequinColorInner=@n2.Base Color
    ShirtMask=NamedRerouteUsage()
   ) {materialFunction=&/MF_ElegantLilyAnkleSequins}.Result) {name=ElegantLilyAnkle_MA, nodeColor=<rgba 0 0.4609 1 1>}
  NamedRerouteDeclaration(None=@n1.Fresnel) {name=InputFresnel, nodeColor=<rgba 0.7812 1 0 1>}
  n5 = ComponentMask(None=VertexInterpolator(VS=LocalPosition() {localOrigin=InstancePreSkinning}.XYZ).PS) {b=true}
  NamedRerouteDeclaration(None=Saturate(None=Add(A=@n5) {constB=-100})) {name=ShirtSequinsMask, nodeColor=<rgba 0 1 0.9766 1>}
  n6 = TextureSample() {
   samplerSource=SSM_Wrap_WorldGroupSettings
   texture=/BRCosmetics/Characters/Player/Female/Medium/Bodies/F_MED_ElegantLilyAnkle/Textures/T_ElegantLilyAnkle_Body_FX
  }
  NamedRerouteDeclaration(None=@n6.G) {name=SequinMask, nodeColor=<rgba 0.9873 0.9341 1 1>}
 ]
}
