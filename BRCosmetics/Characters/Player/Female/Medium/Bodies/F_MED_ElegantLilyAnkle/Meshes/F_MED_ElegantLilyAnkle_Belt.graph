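# Physics asset for the belt mesh of the F_MED_ElegantLilyAnkle female player cosmetic body, with capsule bodies for the belt and front-center skirt bones.
# The belt chain runs from Belt_Base, locked to a corrective bone, through three segments with tight swing limits and zero twist limit.
# A separate pair of larger front skirt capsules is joined by a freely rotating constraint, with no constraint tying it to the belt.
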PhysicsAsset /BRCosmetics/Characters/Player/Female/Medium/Bodies/F_MED_ElegantLilyAnkle/Meshes/F_MED_ElegantLilyAnkle_Belt
bodies {
 Belt_01 [{shapeType=Capsule, length=0.505, radius=0.505, shapeName=Belt_01_capsule, rotation=<rot 0 -90 0>}]
 Belt_02 [{shapeType=Capsule, length=2.7541, radius=2.1844, shapeName=Belt_02_capsule, center=<0.3707 0.7733 -0.3766>, rotation=<rot 90 0 0>}]
 Belt_03 [{
   shapeType=Capsule
   length=2.5385
   radius=2.8707
   shapeName=Belt_03_capsule
   center=<0.8766 0.0442 -0.3697>
   rotation=<rot -14.7263 -75.0898 -80.3427>
  }]
 Belt_Base [{shapeType=Capsule, length=0.505, radius=0.505, shapeName=Belt_Base_capsule, rotation=<rot 0 -90 0>}]
 Belt_corrective_c [{shapeType=Capsule, length=1, radius=1, shapeName=Belt_corrective_c_capsule}]
 dyn_skirt_front_c_01 [{shapeType=Capsule, length=22.9303, radius=7, shapeName=dyn_skirt_front_c_01_capsule, center=<1.6329 -3.9398 0>}]
 dyn_skirt_front_c_02 [{shapeType=Capsule, length=25.7991, radius=7, shapeName=dyn_skirt_front_c_02_capsule, center=<0 -4.4328 0>}]
}
constraints [
 {
  bone1Name=dyn_skirt_front_c_02
  bone2Name=dyn_skirt_front_c_01
  swing1LimitDegrees=45
  swing1Motion=Free
  swing2LimitDegrees=45
  swing2Motion=Free
  twistLimitDegrees=45
  twistMotion=Free
 }
 {bone1Name=Belt_01, bone2Name=Belt_Base, swing1LimitDegrees=25, swing2LimitDegrees=1, twistLimitDegrees=0}
 {bone1Name=Belt_02, bone2Name=Belt_01, swing1LimitDegrees=5, swing2LimitDegrees=15, twistLimitDegrees=0}
 {bone1Name=Belt_03, bone2Name=Belt_02, swing1LimitDegrees=15, swing2LimitDegrees=15, twistLimitDegrees=0}
 {
  bone1Name=Belt_Base
  bone2Name=Belt_corrective_c
  swing1LimitDegrees=45
  swing1Motion=Locked
  swing2LimitDegrees=45
  swing2Motion=Locked
  twistLimitDegrees=45
  twistMotion=Locked
 }
]
