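# Physics asset for the skirt of the female medium ElegantLilyAnkle cosmetic body, defining capsule bodies and joints for its dynamic skirt bones.
# Five chains (back center, left, right and both sides) each run from a tiny Base capsule locked angularly to the pelvis through two larger segments.
# Segment joints carry individually tuned swing angles, with near-zero twist.
# Larger pelvis and thigh capsules form the body volume, with thighs also angularly locked to the pelvis.
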PhysicsAsset /BRCosmetics/Characters/Player/Female/Medium/Bodies/F_MED_ElegantLilyAnkle/Meshes/F_MED_ElegantLilyAnkle_Skirt
bodies {
 dyn_skirt_back_c_01 [{shapeType=Capsule, length=8, radius=6, shapeName=dyn_skirt_back_c_01_capsule, center=<3.2748 5 0>}]
 dyn_skirt_back_c_02 [{shapeType=Capsule, length=8, radius=6, shapeName=dyn_skirt_back_c_02_capsule, center=<4.4105 5 0>}]
 dyn_skirt_back_c_Base [{shapeType=Capsule, length=0.505, radius=0.505, shapeName=dyn_skirt_back_c_Base_capsule}]
 dyn_skirt_back_l_01 [{shapeType=Capsule, length=8, radius=6, shapeName=dyn_skirt_back_l_01_capsule, center=<2.5201 5 0>}]
 dyn_skirt_back_l_02 [{shapeType=Capsule, length=8, radius=6, shapeName=dyn_skirt_back_l_02_capsule, center=<3.3869 5 0>}]
 dyn_skirt_back_l_Base [{shapeType=Capsule, length=0.505, radius=0.505, shapeName=dyn_skirt_back_l_Base_capsule}]
 dyn_skirt_back_r_01 [{shapeType=Capsule, length=8, radius=6, shapeName=dyn_skirt_back_r_01_capsule, center=<1.976 5 0>}]
 dyn_skirt_back_r_02 [{shapeType=Capsule, length=20.7233, radius=6, shapeName=dyn_skirt_back_r_02_capsule, center=<3.3588 5 0>}]
 dyn_skirt_back_r_Base [{shapeType=Capsule, length=0.505, radius=0.505, shapeName=dyn_skirt_back_r_Base_capsule}]
 dyn_skirt_side_l_01 [{shapeType=Capsule, length=8, radius=6, shapeName=dyn_skirt_side_l_01_capsule, center=<0 5 0>}]
 dyn_skirt_side_l_02 [{shapeType=Capsule, length=18.3981, radius=6, shapeName=dyn_skirt_side_l_02_capsule, center=<1.8686 3 0.2418>, rotation=<rot 0 0 11.7288>}]
 dyn_skirt_side_l_Base [{shapeType=Capsule, length=0.505, radius=0.505, shapeName=dyn_skirt_side_l_Base_capsule}]
 dyn_skirt_side_r_01 [{shapeType=Capsule, length=8, radius=6, shapeName=dyn_skirt_side_r_01_capsule, center=<0 6 0>}]
 dyn_skirt_side_r_02 [{shapeType=Capsule, length=22.7642, radius=6, shapeName=dyn_skirt_side_r_02_capsule, center=<0 5 0>}]
 dyn_skirt_side_r_Base [{shapeType=Capsule, length=0.505, radius=0.505, shapeName=dyn_skirt_side_r_Base_capsule}]
 pelvis [{shapeType=Capsule, length=11.1873, radius=10.3323, shapeName=pelvis_capsule, center=<-1.376 -2.1992 0>}]
 thigh_l [{
   shapeType=Capsule
   length=16.5493
   radius=8.6193
   shapeName=thigh_l_capsule
   center=<-18.2341 -1.2814 -1.0283>
   rotation=<rot 86.4323 -5.0402 0.0002>
  }]
 thigh_r [{
   shapeType=Capsule
   length=16.5493
   radius=8.6193
   shapeName=thigh_l_capsule
   center=<18.2206 1.3906 0.6593>
   rotation=<rot 86.4323 -5.0402 -179.9998>
  }]
}
constraints [
 {
  bone1Name=dyn_skirt_back_c_Base
  bone2Name=pelvis
  swing1LimitDegrees=45
  swing1Motion=Locked
  swing2LimitDegrees=45
  swing2Motion=Locked
  twistLimitDegrees=45
  twistMotion=Locked
 }
 {bone1Name=dyn_skirt_back_c_01, bone2Name=dyn_skirt_back_c_Base, swing1LimitDegrees=9, swing2LimitDegrees=5, twistLimitDegrees=0}
 {bone1Name=dyn_skirt_back_c_02, bone2Name=dyn_skirt_back_c_01, swing1LimitDegrees=25, swing2LimitDegrees=10, twistLimitDegrees=0}
 {
  bone1Name=dyn_skirt_back_l_Base
  bone2Name=pelvis
  swing1LimitDegrees=45
  swing1Motion=Locked
  swing2LimitDegrees=45
  swing2Motion=Locked
  twistLimitDegrees=45
  twistMotion=Locked
 }
 {bone1Name=dyn_skirt_back_l_01, bone2Name=dyn_skirt_back_l_Base, swing1LimitDegrees=5, swing2LimitDegrees=15, twistLimitDegrees=0}
 {bone1Name=dyn_skirt_back_l_02, bone2Name=dyn_skirt_back_l_01, swing1LimitDegrees=25, swing2LimitDegrees=10, twistLimitDegrees=0}
 {
  bone1Name=dyn_skirt_back_r_Base
  bone2Name=pelvis
  swing1LimitDegrees=45
  swing1Motion=Locked
  swing2LimitDegrees=45
  swing2Motion=Locked
  twistLimitDegrees=45
  twistMotion=Locked
 }
 {bone1Name=dyn_skirt_back_r_01, bone2Name=dyn_skirt_back_r_Base, swing1LimitDegrees=5, swing2LimitDegrees=15, twistLimitDegrees=0}
 {bone1Name=dyn_skirt_back_r_02, bone2Name=dyn_skirt_back_r_01, swing1LimitDegrees=25, swing2LimitDegrees=5, twistLimitDegrees=0}
 {
  bone1Name=dyn_skirt_side_l_Base
  bone2Name=pelvis
  swing1LimitDegrees=45
  swing1Motion=Locked
  swing2LimitDegrees=45
  swing2Motion=Locked
  twistLimitDegrees=45
  twistMotion=Locked
 }
 {bone1Name=dyn_skirt_side_l_02, bone2Name=dyn_skirt_side_l_01, swing1LimitDegrees=25, swing2LimitDegrees=0, twistLimitDegrees=0}
 {bone1Name=dyn_skirt_side_l_01, bone2Name=dyn_skirt_side_l_Base, swing1LimitDegrees=25, swing2LimitDegrees=0, twistLimitDegrees=0}
 {
  bone1Name=dyn_skirt_side_r_Base
  bone2Name=pelvis
  swing1LimitDegrees=45
  swing1Motion=Locked
  swing2LimitDegrees=45
  swing2Motion=Locked
  twistLimitDegrees=45
  twistMotion=Locked
 }
 {bone1Name=dyn_skirt_side_r_02, bone2Name=dyn_skirt_side_r_01, swing1LimitDegrees=10, swing2LimitDegrees=0, twistLimitDegrees=10}
 {bone1Name=dyn_skirt_side_r_01, bone2Name=dyn_skirt_side_r_Base, swing1LimitDegrees=10, swing2LimitDegrees=0, twistLimitDegrees=0}
 {
  bone1Name=thigh_l
  bone2Name=pelvis
  swing1LimitDegrees=45
  swing1Motion=Locked
  swing2LimitDegrees=45
  swing2Motion=Locked
  twistLimitDegrees=45
  twistMotion=Locked
 }
 {
  bone1Name=thigh_r
  bone2Name=pelvis
  swing1LimitDegrees=45
  swing1Motion=Locked
  swing2LimitDegrees=45
  swing2Motion=Locked
  twistLimitDegrees=45
  twistMotion=Locked
 }
]
