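# Physics asset for the straps of the female medium ElegantLilyAnkle body, with capsule bodies on strap, front skirt, skirt panel and leg bones.
# Left and right strap chains use mostly small swing and twist limits, with front chains hanging from Master_skirt bones and back chains from bases locked to root.
# Front and back chain ends are constrained together, closing a loop on each side.
# Skirt panel and leg capsules mostly use locked joints, except free thigh_l-to-root.
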PhysicsAsset /BRCosmetics/Characters/Player/Female/Medium/Bodies/F_MED_ElegantLilyAnkle/Meshes/F_MED_ElegantLilyAnkle_strap_FRONTE
bodies {
 Master_skirt_l_front [{
   shapeType=Capsule
   length=3.8128
   radius=1.2424
   shapeName=Master_skirt_l_front_capsule
   center=<2.4003 0.8083 0.4882>
   rotation=<rot 0 92.7516 -81.8405>
  }]
 Master_skirt_l_front_02 [{
   shapeType=Capsule
   length=3.5915
   radius=1.5659
   shapeName=Master_skirt_l_front_02_capsule
   center=<5.6755 1.5432 -0.1694>
   rotation=<rot 88.2902 -0.0004 -0.0004>
  }]
 Master_skirt_r_front [{shapeType=Capsule, length=1, radius=1}]
 Master_skirt_r_front_01 [{
   shapeType=Capsule
   length=2.766
   radius=1.6701
   shapeName=Master_skirt_r_front_01_capsule
   center=<3.7028 1.1959 -0.5851>
   rotation=<rot -1.7329 -86.9724 -82.114>
  }]
 Master_skirt_r_front_02 [{
   shapeType=Capsule
   length=4.0729
   radius=1.3382
   shapeName=Master_skirt_r_front_02_capsule
   center=<2.5673 0.5123 0.0594>
   rotation=<rot 0 -88.4493 -88.3686>
  }]
 calf_l [{
   shapeType=Capsule
   length=31.4452
   radius=7.9941
   shapeName=calf_l_capsule
   center=<-24.282 -3.0654 1.055>
   rotation=<rot -90 -166.8094 168.2521>
  }]
 calf_r [{
   shapeType=Capsule
   length=29.2346
   radius=7.1397
   shapeName=calf_r_capsule
   center=<21.3715 3.7306 -0.0381>
   rotation=<rot 89.8978 0.0036 0.0038>
  }]
 dyn_skirt_back_l_02 [{shapeType=Capsule, length=8.8142, radius=11.7286, shapeName=dyn_skirt_back_l_02_capsule, center=<0 -11.4448 0>}]
 dyn_skirt_back_r_02 [
  {shapeType=Capsule, length=7.5469, radius=8.4949, shapeName=dyn_skirt_back_r_02_capsule, center=<0 -7.2693 -1.8668>}
  {
   shapeType=Capsule
   length=3.6221
   radius=7.8078
   shapeName=dyn_skirt_back_r_02_capsule
   center=<0.8506 -12.1699 5.0214>
   rotation=<rot 4.5531 -7.0143 -57.168>
  }
 ]
 dyn_skirt_front_l_02 [{
   shapeType=Capsule
   length=0.8895
   radius=11.5692
   shapeName=dyn_skirt_front_l_02_capsule
   center=<2.0838 -10.0514 0>
   rotation=<rot 0 0.0001 -9.5962>
  }]
 dyn_skirt_front_l_Base [{shapeType=Capsule, length=0.505, radius=0.505, shapeName=dyn_skirt_front_l_Base_capsule, rotation=<rot 0 -90 0>}]
 dyn_skirt_front_r_02 [{
   shapeType=Capsule
   length=24.0904
   radius=9.6761
   shapeName=dyn_skirt_front_r_02_capsule
   center=<-1.808 -8.5769 2.1943>
   rotation=<rot 89.8739 0.0035 -2.2281>
  }]
 dyn_skirt_front_r_Base [{shapeType=Capsule, length=0.505, radius=0.505, shapeName=dyn_skirt_front_l_Base_capsule, rotation=<rot 0 -90 180>}]
 dyn_skirt_side_l_02 [{
   shapeType=Capsule
   length=13.2821
   radius=8.0579
   shapeName=dyn_skirt_side_l_02_capsule
   center=<-3.9059 -9.3838 1.9582>
   rotation=<rot 90 -180 180>
  }]
 dyn_skirt_side_r_01 [{
   shapeType=Capsule
   length=11.7149
   radius=5.0151
   shapeName=dyn_skirt_side_r_01_capsule
   center=<5.2258 -4.639 -0.0531>
   rotation=<rot 0.8602 95.5079 -94.2411>
  }]
 dyn_skirt_side_r_02 [{
   shapeType=Capsule
   length=5.4353
   radius=10.6388
   shapeName=dyn_skirt_side_r_02_capsule
   center=<-2.27 -10.246 -2.5461>
   rotation=<rot -6.4556 -90 -94.6375>
  }]
 dyn_strap_back_l_01 [{shapeType=Capsule, length=7.1449, radius=1.3383, center=<4.3668 0.4711 0.1161>, rotation=<rot 88.4769 -179.9995 -176.7658>}]
 dyn_strap_back_l_02 [{
   shapeType=Capsule
   length=6.945
   radius=1.2578
   shapeName=dyn_strap_back_l_02_capsule
   center=<3.5597 -0.132 0>
   rotation=<rot 88.7349 -0.1122 -0.0003>
  }]
 dyn_strap_back_l_03 [{
   shapeType=Capsule
   length=5.2943
   radius=1.1701
   shapeName=dyn_strap_back_l_03_capsule
   center=<2.9899 0.226 0>
   rotation=<rot 90 180 175.6772>
  }]
 dyn_strap_back_l_04 [{shapeType=Capsule, length=4.1961, radius=1.3126, shapeName=dyn_strap_back_l_04_capsule, center=<0.9611 0 0>, rotation=<rot 90 0 0>}]
 dyn_strap_back_l_Base [{shapeType=Capsule, length=1, radius=1}]
 dyn_strap_back_l_end [{shapeType=Capsule, length=2.7551, radius=1.5639, shapeName=dyn_strap_back_l_end_capsule, rotation=<rot 90 0 0>}]
 dyn_strap_back_r_01 [{
   shapeType=Capsule
   length=8.0164
   radius=1.3977
   shapeName=dyn_strap_back_l_01_capsule
   center=<4.4 0 0>
   rotation=<rot 90 150.5107 -29.4893>
  }]
 dyn_strap_back_r_02 [{shapeType=Capsule, length=7.049, radius=1.2119, shapeName=dyn_strap_back_l_02_capsule, center=<3.26 0 0>, rotation=<rot 90 0 -179.8882>}]
 dyn_strap_back_r_03 [{
   shapeType=Capsule
   length=6.6587
   radius=1.6721
   shapeName=dyn_strap_back_l_03_capsule
   center=<3.8681 -0.038 0>
   rotation=<rot 90 -179.8881 0.6748>
  }]
 dyn_strap_back_r_04 [{shapeType=Capsule, length=5.5497, radius=1.5921, shapeName=dyn_strap_back_l_04_capsule, center=<3.0282 0 0>, rotation=<rot 90 180 0>}]
 dyn_strap_back_r_Base [{shapeType=Capsule, length=0.505, radius=0.505, shapeName=dyn_strap_back_r_Base_capsule, rotation=<rot 0 -90 0>}]
 dyn_strap_back_r_end [{shapeType=Capsule, length=2.6477, radius=1.0931, shapeName=dyn_strap_back_r_end_capsule, rotation=<rot 90 -169.9693 -177.1518>}]
 dyn_strap_front_l_01 [{shapeType=Capsule, length=6.1234, radius=1.3258, center=<6.06 0.3573 0.0132>, rotation=<rot 89.8753 -179.9926 -176.4545>}]
 dyn_strap_front_l_02 [{
   shapeType=Capsule
   length=1.7223
   radius=1.4133
   shapeName=dyn_strap_front_l_02_capsule
   center=<1.2674 0.3746 0>
   rotation=<rot 90 180 175.6588>
  }]
 dyn_strap_front_l_03 [{
   shapeType=Capsule
   length=3.8295
   radius=1.5882
   shapeName=dyn_strap_front_l_03_capsule
   center=<1.5556 0.2833 0>
   rotation=<rot 90 -176.4237 173.2548>
  }]
 dyn_strap_front_l_04 [{
   shapeType=Capsule
   length=3.2497
   radius=1.6321
   shapeName=dyn_strap_front_l_04_capsule
   center=<2.391 0 0>
   rotation=<rot 6.6527 -90 91.989>
  }]
 dyn_strap_front_l_05 [{shapeType=Capsule, length=1.2872, radius=1.8661, shapeName=dyn_strap_front_l_05_capsule, rotation=<rot 0.281 96.2074 -91.7649>}]
 dyn_strap_front_l_Base [{shapeType=Capsule, length=0.505, radius=0.505, shapeName=dyn_strap_front_l_Base_capsule, rotation=<rot 0 -90 0>}]
 dyn_strap_front_r_01 [{shapeType=Capsule, length=8.4536, radius=1.4036, center=<5.4981 0.4962 -0.009>, rotation=<rot -89.9066 -179.9977 179.9977>}]
 dyn_strap_front_r_02 [{
   shapeType=Capsule
   length=3.6242
   radius=0.964
   shapeName=dyn_strap_front_l_02_capsule
   center=<1.2789 0.0468 0>
   rotation=<rot 90 180 -2.0954>
  }]
 dyn_strap_front_r_03 [{shapeType=Capsule, length=4.2328, radius=1.1849, shapeName=dyn_strap_front_l_03_capsule, center=<2.1005 0.1144 0>, rotation=<rot 90 180 0>}]
 dyn_strap_front_r_04 [{shapeType=Capsule, length=7.2022, radius=1.6156, shapeName=dyn_strap_front_r_04_capsule, center=<3.4894 0.0891 0>, rotation=<rot 90 0 0>}]
 dyn_strap_front_r_05 [{shapeType=Capsule, length=3.1461, radius=1.364, shapeName=dyn_strap_front_r_05_capsule, rotation=<rot 90 0 0>}]
 dyn_strap_front_r_base [{shapeType=Capsule, length=1, radius=1}]
 root [{shapeType=Capsule, length=0.505, radius=0.505, shapeName=root_capsule, rotation=<rot 0 -90 0>}]
 thigh_l [{
   shapeType=Capsule
   length=18.8177
   radius=8.8879
   shapeName=thigh_l_capsule
   center=<-33.6252 -2.1644 1.0768>
   rotation=<rot 0 -84.5649 -87.8134>
  }]
 thigh_twist_01_l [{
   shapeType=Capsule
   length=9.3632
   radius=15.008
   shapeName=thigh_twist_01_l_capsule
   center=<11.6949 -3.4787 2.8071>
   rotation=<rot 83.226 173.9645 180>
  }]
 thigh_twist_01_r [
  {
   shapeType=Capsule
   length=16.5184
   radius=7.6593
   shapeName=thigh_twist_01_l_capsule
   center=<4.1752 -0.0023 -0.2364>
   rotation=<rot 85.2622 -0.0046 -179.9724>
  }
  {
   shapeType=Capsule
   length=9.3632
   radius=12.9614
   shapeName=thigh_twist_01_l_capsule
   center=<-11.6949 3.4787 -2.8071>
   rotation=<rot 83.226 173.9645 0>
  }
 ]
}
constraints [
 {
  bone1Name=dyn_strap_front_l_01
  bone2Name=dyn_strap_front_l_Base
  swing1LimitDegrees=20
  swing2LimitDegrees=20
  twistLimitDegrees=0
  twistMotion=Locked
 }
 {bone1Name=dyn_strap_front_l_02, bone2Name=dyn_strap_front_l_01, swing1LimitDegrees=25, swing2LimitDegrees=25, twistLimitDegrees=0}
 {bone1Name=dyn_strap_front_l_03, bone2Name=dyn_strap_front_l_02, swing1LimitDegrees=25, swing2LimitDegrees=25, twistLimitDegrees=0}
 {bone1Name=dyn_strap_back_l_01, bone2Name=dyn_strap_back_l_Base, swing1LimitDegrees=20, swing2LimitDegrees=20, twistLimitDegrees=4}
 {bone1Name=dyn_strap_back_l_02, bone2Name=dyn_strap_back_l_01, swing1LimitDegrees=20, swing2LimitDegrees=25, twistLimitDegrees=4}
 {bone1Name=dyn_strap_back_l_03, bone2Name=dyn_strap_back_l_02, swing1LimitDegrees=25, swing2LimitDegrees=25, twistLimitDegrees=4}
 {
  bone1Name=dyn_strap_back_l_04
  bone2Name=dyn_strap_back_l_03
  swing1LimitDegrees=25
  swing1Motion=Locked
  swing2LimitDegrees=25
  swing2Motion=Locked
  twistLimitDegrees=4
 }
 {bone1Name=dyn_strap_front_l_Base, bone2Name=Master_skirt_l_front_02, swing1LimitDegrees=20, swing2LimitDegrees=20, twistLimitDegrees=1}
 {
  bone1Name=dyn_strap_back_l_Base
  bone2Name=root
  swing1LimitDegrees=45
  swing1Motion=Locked
  swing2LimitDegrees=45
  swing2Motion=Locked
  twistLimitDegrees=45
  twistMotion=Locked
 }
 {
  bone1Name=dyn_skirt_front_l_02
  bone2Name=dyn_skirt_front_l_Base
  swing1LimitDegrees=45
  swing1Motion=Locked
  swing2LimitDegrees=45
  swing2Motion=Locked
  twistLimitDegrees=45
  twistMotion=Locked
 }
 {
  bone1Name=dyn_skirt_back_l_02
  bone2Name=root
  swing1LimitDegrees=45
  swing1Motion=Locked
  swing2LimitDegrees=45
  swing2Motion=Locked
  twistLimitDegrees=45
  twistMotion=Locked
 }
 {
  bone1Name=dyn_skirt_front_l_Base
  bone2Name=thigh_l
  swing1LimitDegrees=45
  swing1Motion=Locked
  swing2LimitDegrees=45
  swing2Motion=Locked
  twistLimitDegrees=45
  twistMotion=Locked
 }
 {
  bone1Name=thigh_twist_01_l
  bone2Name=thigh_l
  swing1LimitDegrees=45
  swing1Motion=Locked
  swing2LimitDegrees=45
  swing2Motion=Locked
  twistLimitDegrees=45
  twistMotion=Locked
 }
 {bone1Name=dyn_strap_back_r_02, bone2Name=dyn_strap_back_r_01, swing1LimitDegrees=25, swing2LimitDegrees=25, twistLimitDegrees=4}
 {bone1Name=dyn_strap_back_r_03, bone2Name=dyn_strap_back_r_02, swing1LimitDegrees=25, swing2LimitDegrees=25, twistLimitDegrees=3}
 {bone1Name=dyn_strap_back_r_04, bone2Name=dyn_strap_back_r_03, swing1LimitDegrees=20, swing2LimitDegrees=25, twistLimitDegrees=4}
 {bone1Name=dyn_strap_front_r_02, bone2Name=dyn_strap_front_r_01, swing1LimitDegrees=15, swing2LimitDegrees=1, twistLimitDegrees=0}
 {bone1Name=dyn_strap_front_r_03, bone2Name=dyn_strap_front_r_02, swing1LimitDegrees=10, swing2LimitDegrees=10, twistLimitDegrees=0}
 {
  bone1Name=dyn_skirt_front_r_Base
  bone2Name=root
  swing1LimitDegrees=45
  swing1Motion=Locked
  swing2LimitDegrees=45
  swing2Motion=Locked
  twistLimitDegrees=45
  twistMotion=Locked
 }
 {
  bone1Name=dyn_skirt_front_r_02
  bone2Name=dyn_skirt_front_r_Base
  swing1LimitDegrees=45
  swing1Motion=Locked
  swing2LimitDegrees=45
  swing2Motion=Locked
  twistLimitDegrees=45
  twistMotion=Locked
 }
 {
  bone1Name=thigh_twist_01_r
  bone2Name=root
  swing1LimitDegrees=45
  swing1Motion=Locked
  swing2LimitDegrees=45
  swing2Motion=Locked
  twistLimitDegrees=45
  twistMotion=Locked
 }
 {
  bone1Name=dyn_strap_front_l_04
  bone2Name=dyn_strap_front_l_03
  swing1LimitDegrees=25
  swing1Motion=Locked
  swing2LimitDegrees=25
  swing2Motion=Locked
  twistLimitDegrees=0
 }
 {bone1Name=dyn_strap_front_l_05, bone2Name=dyn_strap_front_l_04, swing1LimitDegrees=20, swing2LimitDegrees=20, twistLimitDegrees=1}
 {
  bone1Name=dyn_strap_front_r_04
  bone2Name=dyn_strap_front_r_03
  swing1LimitDegrees=25
  swing2LimitDegrees=25
  twistLimitDegrees=0
  twistMotion=Locked
 }
 {bone1Name=dyn_strap_front_r_05, bone2Name=dyn_strap_front_r_04, swing1LimitDegrees=20, swing2LimitDegrees=25, twistLimitDegrees=0}
 {bone1Name=dyn_strap_back_r_01, bone2Name=dyn_strap_back_r_Base, swing1LimitDegrees=20, swing2LimitDegrees=22, twistLimitDegrees=4}
 {
  bone1Name=dyn_strap_back_r_Base
  bone2Name=root
  swing1LimitDegrees=45
  swing1Motion=Locked
  swing2LimitDegrees=45
  swing2Motion=Locked
  twistLimitDegrees=45
  twistMotion=Locked
 }
 {
  bone1Name=dyn_strap_back_r_end
  bone2Name=dyn_strap_back_r_04
  swing1LimitDegrees=45
  swing1Motion=Locked
  swing2LimitDegrees=45
  swing2Motion=Locked
  twistLimitDegrees=45
  twistMotion=Locked
 }
 {
  bone1Name=dyn_strap_front_r_05
  bone2Name=dyn_strap_back_r_end
  swing1LimitDegrees=45
  swing1Motion=Locked
  swing2LimitDegrees=45
  swing2Motion=Locked
  twistLimitDegrees=1
 }
 {bone1Name=dyn_strap_front_r_01, bone2Name=dyn_strap_front_r_base, swing1LimitDegrees=10, swing2LimitDegrees=5, twistLimitDegrees=1}
 {
  bone1Name=dyn_strap_front_r_base
  bone2Name=Master_skirt_r_front_02
  swing1LimitDegrees=5
  swing1Motion=Locked
  swing2LimitDegrees=15
  swing2Motion=Locked
  twistLimitDegrees=1
  twistMotion=Locked
 }
 {bone1Name=dyn_strap_back_l_end, bone2Name=dyn_strap_back_l_04, swing1LimitDegrees=20, swing2LimitDegrees=20, twistLimitDegrees=4}
 {
  bone1Name=dyn_strap_back_l_end
  bone2Name=dyn_strap_front_l_05
  swing1LimitDegrees=25
  swing1Motion=Locked
  swing2LimitDegrees=25
  swing2Motion=Locked
  twistLimitDegrees=45
  twistMotion=Free
 }
 {
  bone1Name=dyn_skirt_back_r_02
  bone2Name=root
  swing1LimitDegrees=45
  swing1Motion=Locked
  swing2LimitDegrees=45
  swing2Motion=Locked
  twistLimitDegrees=45
  twistMotion=Locked
 }
 {
  bone1Name=dyn_skirt_side_l_02
  bone2Name=root
  swing1LimitDegrees=45
  swing1Motion=Locked
  swing2LimitDegrees=45
  swing2Motion=Locked
  twistLimitDegrees=45
  twistMotion=Locked
 }
 {
  bone1Name=dyn_skirt_side_r_02
  bone2Name=dyn_skirt_side_r_01
  swing1LimitDegrees=45
  swing1Motion=Locked
  swing2LimitDegrees=45
  swing2Motion=Locked
  twistLimitDegrees=45
  twistMotion=Locked
 }
 {
  bone1Name=Master_skirt_r_front_02
  bone2Name=Master_skirt_r_front_01
  swing1LimitDegrees=10
  swing2LimitDegrees=5
  twistLimitDegrees=45
  twistMotion=Locked
 }
 {
  bone1Name=Master_skirt_r_front
  bone2Name=dyn_skirt_front_r_Base
  swing1LimitDegrees=5
  swing1Motion=Locked
  swing2LimitDegrees=20
  swing2Motion=Locked
  twistLimitDegrees=45
  twistMotion=Locked
 }
 {bone1Name=Master_skirt_l_front_02, bone2Name=Master_skirt_l_front, swing1LimitDegrees=20, swing2LimitDegrees=20, twistLimitDegrees=1}
 {
  bone1Name=Master_skirt_l_front
  bone2Name=dyn_skirt_front_l_Base
  swing1LimitDegrees=45
  swing1Motion=Locked
  swing2LimitDegrees=45
  swing2Motion=Locked
  twistLimitDegrees=45
  twistMotion=Locked
 }
 {
  bone1Name=calf_l
  bone2Name=thigh_l
  swing1LimitDegrees=45
  swing1Motion=Locked
  swing2LimitDegrees=45
  swing2Motion=Locked
  twistLimitDegrees=45
  twistMotion=Locked
 }
 {
  bone1Name=calf_r
  bone2Name=root
  swing1LimitDegrees=45
  swing1Motion=Locked
  swing2LimitDegrees=45
  swing2Motion=Locked
  twistLimitDegrees=45
  twistMotion=Locked
 }
 {bone1Name=Master_skirt_r_front_01, bone2Name=Master_skirt_r_front, swing1LimitDegrees=15, swing2LimitDegrees=15, twistLimitDegrees=1}
 {
  bone1Name=dyn_skirt_side_r_01
  bone2Name=root
  swing1LimitDegrees=45
  swing1Motion=Locked
  swing2LimitDegrees=45
  swing2Motion=Locked
  twistLimitDegrees=45
  twistMotion=Locked
 }
 {
  bone1Name=thigh_l
  bone2Name=root
  swing1LimitDegrees=45
  swing1Motion=Free
  swing2LimitDegrees=45
  swing2Motion=Free
  twistLimitDegrees=45
  twistMotion=Free
 }
]
