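# Anim Blueprint for the female ElegantLilyAnkle cosmetic body, adding secondary motion to skirt, straps, bows and belt.
# The parent mesh pose runs through a body Control Rig, skirt AnimDynamics, strap Control Rig and blended strap RigidBody simulations.
# Spine-weighted constraints then position torso bows before AnimDynamics chains move bows and belt, using spherical limits around the feet.
# The EventGraph feeds the rigs movement, jump, front-end, alt-skin and Lily emote-montage state.
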AnimBlueprint /BRCosmetics/Characters/Player/Female/Medium/Bodies/F_MED_ElegantLilyAnkle/Meshes/F_MED_ElegantLilyAnkle_AnimBP
target_skeleton=&/F_MED_ElegantLilyAnkle_Skeleton
graphs [
 AnimGraph {nodes [
   RigidBody!(ComponentPose, Alpha, OverrideWorldGravity:Vector="0,0,0", ExternalForce) -> (Pose) {
    bClampLinearTranslationLimitToRefPose=true
    bOverrideWorldGravity=true
    overridePhysicsAsset=&/F_MED_ElegantLilyAnkle_Belt
    simulationSpace=BaseBoneSpace
    alphaBoolBlend {blendOption=Linear}
    baseBoneRef {boneName=pelvis}
   }
   RigidBody(OverrideWorldGravity="0,0,0") {
    bOverrideWorldGravity=true
    overridePhysicsAsset=&/F_MED_ElegantLilyAnkle_Skirt
    simulationSpace=BaseBoneSpace
    alphaBoolBlend {blendOption=Linear}
    baseBoneRef {boneName=pelvis}
    componentAppliedLinearAccClamp=<3000 3000 3000>
    componentLinearAccScale=<0.3 0.3 0.3>
    componentLinearVelScale=<0.3 0.3 0.3>
   }
   n1 = "Constraint - dyn_torso_bow__r_base7 (PPPPPP)"!(
    ComponentPose=Local To Component(LocalPose=Use cached pose 'RigidBody'())
    Alpha=1
    ConstraintWeights_0=0.0324
    ConstraintWeights_1:float=0.0512
    ConstraintWeights_2:float=0.2767
    ConstraintWeights_3:float=0.1314
    ConstraintWeights_4:float=0.4717
    ConstraintWeights_5:float=0.0366
   ) -> (Pose) {
    alphaBoolBlend {blendOption=Linear}
    boneToModify {boneName=dyn_torso_bow__r_base7}
    constraintSetup [
     {targetBone {boneName=pelvis}}
     {targetBone {boneName=spine_01}}
     {targetBone {boneName=spine_02}}
     {targetBone {boneName=spine_03}}
     {targetBone {boneName=spine_04}}
     {targetBone {boneName=spine_05}}
    ]
    constraintWeights [1,1,1,1,1,1]
   }
   n2 = "Constraint - dyn_torso_bow__l_base (PPPPPP)"!(
    ComponentPose=@n1
    Alpha=1
    ConstraintWeights_0=0.0324
    ConstraintWeights_1:float=0.0512
    ConstraintWeights_2:float=0.2767
    ConstraintWeights_3:float=0.1314
    ConstraintWeights_4:float=0.4717
    ConstraintWeights_5:float=0.0366
   ) -> (Pose) {
    alphaBoolBlend {blendOption=Linear}
    boneToModify {boneName=dyn_torso_bow__l_base}
    constraintSetup [
     {targetBone {boneName=pelvis}}
     {targetBone {boneName=spine_01}}
     {targetBone {boneName=spine_02}}
     {targetBone {boneName=spine_03}}
     {targetBone {boneName=spine_04}}
     {targetBone {boneName=spine_05}}
    ]
    constraintWeights [1,1,1,1,1,1]
   }
   n3 = "Transform (Modify) Bone - Bone: dyn_torso_bow__r_base7"(ComponentPose=@n2, Alpha=1, Translation="-0.5,0,0", Rotation="0,0,0", Scale="1,1,1") {translationMode=BMM_Additive, alphaBoolBlend {blendOption=Linear}, boneToModify {boneName=dyn_torso_bow__r_base7}, translation {x=-0.5}}
   n4 = "Transform (Modify) Bone - Bone: dyn_torso_bow__l_base"(ComponentPose=@n3, Alpha=1, Translation="0.5,0,0", Rotation="0,0,0", Scale="1,1,1") {translationMode=BMM_Additive, alphaBoolBlend {blendOption=Linear}, boneToModify {boneName=dyn_torso_bow__l_base}, translation {x=0.5}}
   n5 = Anim Dynamics - Bone: dyn_torso_bow__l_01!(ComponentPose=@n4, Alpha=1, GravityOverride:Vector="0,0,0", ExternalForce="0,0,0") -> (Pose) {
    angularDampingOverride=0.9999
    angularSpringConstant=2000
    bOverrideAngularDamping=true
    bUseGravityOverride=true
    simulationSpace=RootRelative
    alphaBoolBlend {blendOption=Linear}
    boundBone {boneName=dyn_torso_bow__l_01}
    componentLinearAccScale=<1 1 1>
    componentLinearVelScale=<1 1 1>
    physicsBodyDefinitions [{
      boundBone {boneName=dyn_torso_bow__l_01}
      boxExtents=<10 10 10>
      constraintSetup {angularLimitsMax=<0 5 5>, angularLimitsMin=<0 -20 -20>, angularTarget=<0 0 0>}
      localJointOffset=<0.5 0 0>
     }]
   }
   n6 = Anim Dynamics - Bone: dyn_torso_bow__l_02!(ComponentPose=@n5, Alpha=1, GravityOverride:Vector="0,0,0", ExternalForce="0,0,0") -> (Pose) {
    angularDampingOverride=0.9999
    angularSpringConstant=2000
    bOverrideAngularDamping=true
    bUseGravityOverride=true
    simulationSpace=RootRelative
    alphaBoolBlend {blendOption=Linear}
    boundBone {boneName=dyn_torso_bow__l_02}
    componentLinearAccScale=<1 1 1>
    componentLinearVelScale=<1 1 1>
    physicsBodyDefinitions [{
      boundBone {boneName=dyn_torso_bow__l_02}
      boxExtents=<10 10 10>
      constraintSetup {angularLimitsMax=<0 5 10>, angularLimitsMin=<0 -20 -20>, angularTarget=<0 0 0>}
      localJointOffset=<0.5 0 0>
     }]
   }
   n7 = Anim Dynamics - Bone: dyn_torso_bow__l_03!(ComponentPose=@n6, Alpha=1, GravityOverride:Vector="0,0,0", ExternalForce="0,0,0") -> (Pose) {
    angularDampingOverride=0.9999
    angularSpringConstant=2000
    bOverrideAngularDamping=true
    bUseGravityOverride=true
    simulationSpace=RootRelative
    alphaBoolBlend {blendOption=Linear}
    boundBone {boneName=dyn_torso_bow__l_03}
    componentLinearAccScale=<1 1 1>
    componentLinearVelScale=<1 1 1>
    physicsBodyDefinitions [{
      boundBone {boneName=dyn_torso_bow__l_03}
      boxExtents=<10 10 10>
      constraintSetup {angularLimitsMax=<0 5 20>, angularLimitsMin=<0 -20 -20>, angularTarget=<0 0 0>}
      localJointOffset=<0.5 0 0>
     }]
   }
   n8 = Anim Dynamics - Bone: dyn_torso_bow__l_04!(ComponentPose=@n7, Alpha=1, GravityOverride:Vector="0,0,0", ExternalForce="0,0,0") -> (Pose) {
    angularDampingOverride=0.9999
    angularSpringConstant=2000
    bOverrideAngularDamping=true
    bUseGravityOverride=true
    bUsePlanarLimit=false
    simulationSpace=RootRelative
    alphaBoolBlend {blendOption=Linear}
    boundBone {boneName=dyn_torso_bow__l_04}
    componentLinearAccScale=<1 1 1>
    componentLinearVelScale=<1 1 1>
    physicsBodyDefinitions [{
      boundBone {boneName=dyn_torso_bow__l_04}
      boxExtents=<10 10 10>
      constraintSetup {angularLimitsMax=<0 0 20>, angularLimitsMin=<0 -20 -20>, angularTarget=<0 0 0>}
      localJointOffset=<0.5 0 0>
     }]
   }
   n9 = Anim Dynamics - Bone: dyn_torso_bow__l_05!(ComponentPose=@n8, Alpha=1, GravityOverride:Vector="0,0,0", ExternalForce="0,0,0") -> (Pose) {
    angularDampingOverride=0.9999
    angularSpringConstant=2000
    bOverrideAngularDamping=true
    bUseGravityOverride=true
    simulationSpace=RootRelative
    alphaBoolBlend {blendOption=Linear}
    boundBone {boneName=dyn_torso_bow__l_05}
    componentLinearAccScale=<1 1 1>
    componentLinearVelScale=<1 1 1>
    physicsBodyDefinitions [{
      boundBone {boneName=dyn_torso_bow__l_05}
      boxExtents=<10 10 10>
      constraintSetup {angularLimitsMax=<0 5 20>, angularLimitsMin=<0 -20 -10>, angularTarget=<0 0 0>}
      localJointOffset=<0.5 0 0>
     }]
   }
   n10 = Anim Dynamics - Bone: dyn_torso_bow__l_06!(ComponentPose=@n9, Alpha=1, GravityOverride:Vector="0,0,0", ExternalForce="0,0,0") -> (Pose) {
    angularDampingOverride=0.9999
    angularSpringConstant=2000
    bOverrideAngularDamping=true
    bUseGravityOverride=true
    simulationSpace=RootRelative
    alphaBoolBlend {blendOption=Linear}
    boundBone {boneName=dyn_torso_bow__l_06}
    componentLinearAccScale=<1 1 1>
    componentLinearVelScale=<1 1 1>
    physicsBodyDefinitions [{
      boundBone {boneName=dyn_torso_bow__l_06}
      boxExtents=<10 10 10>
      constraintSetup {angularLimitsMax=<0 5 20>, angularLimitsMin=<0 -20 -20>, angularTarget=<0 0 0>}
      localJointOffset=<0.5 0 0>
     }]
   }
   n11 = Anim Dynamics - Bone: dyn_torso_bow__r_01!(ComponentPose=@n10, Alpha=1, GravityOverride:Vector="0,0,0", ExternalForce="0,0,0") -> (Pose) {
    angularDampingOverride=0.9999
    angularSpringConstant=2000
    bOverrideAngularDamping=true
    bUseGravityOverride=true
    simulationSpace=RootRelative
    alphaBoolBlend {blendOption=Linear}
    boundBone {boneName=dyn_torso_bow__r_01}
    componentLinearAccScale=<1 1 1>
    componentLinearVelScale=<1 1 1>
    physicsBodyDefinitions [{
      boundBone {boneName=dyn_torso_bow__r_01}
      boxExtents=<10 10 10>
      constraintSetup {angularLimitsMax=<0 5 5>, angularLimitsMin=<0 -20 -20>, angularTarget=<0 0 0>}
      localJointOffset=<0.5 0 0>
     }]
   }
   n12 = Anim Dynamics - Bone: dyn_torso_bow__r_02!(ComponentPose=@n11, Alpha=1, GravityOverride:Vector="0,0,0", ExternalForce="0,0,0") -> (Pose) {
    angularDampingOverride=0.9999
    angularSpringConstant=2000
    bOverrideAngularDamping=true
    bUseGravityOverride=true
    simulationSpace=RootRelative
    alphaBoolBlend {blendOption=Linear}
    boundBone {boneName=dyn_torso_bow__r_02}
    componentLinearAccScale=<1 1 1>
    componentLinearVelScale=<1 1 1>
    physicsBodyDefinitions [{
      boundBone {boneName=dyn_torso_bow__r_02}
      boxExtents=<10 10 10>
      constraintSetup {angularLimitsMax=<0 5 10>, angularLimitsMin=<0 -20 -20>, angularTarget=<0 0 0>}
      localJointOffset=<0.5 0 0>
     }]
   }
   n13 = Anim Dynamics - Bone: dyn_torso_bow__r_03!(ComponentPose=@n12, Alpha=1, GravityOverride:Vector="0,0,0", ExternalForce="0,0,0") -> (Pose) {
    angularDampingOverride=0.9999
    angularSpringConstant=2000
    bOverrideAngularDamping=true
    bUseGravityOverride=true
    simulationSpace=RootRelative
    alphaBoolBlend {blendOption=Linear}
    boundBone {boneName=dyn_torso_bow__r_03}
    componentLinearAccScale=<1 1 1>
    componentLinearVelScale=<1 1 1>
    physicsBodyDefinitions [{
      boundBone {boneName=dyn_torso_bow__r_03}
      boxExtents=<10 10 10>
      constraintSetup {angularLimitsMax=<0 5 20>, angularLimitsMin=<0 -20 -20>, angularTarget=<0 0 0>}
      localJointOffset=<0.5 0 0>
     }]
   }
   n14 = Anim Dynamics - Bone: dyn_torso_bow__r_04!(ComponentPose=@n13, Alpha=1, GravityOverride:Vector="0,0,0", ExternalForce="0,0,0") -> (Pose) {
    angularDampingOverride=0.9999
    angularSpringConstant=2000
    bOverrideAngularDamping=true
    bUseGravityOverride=true
    simulationSpace=RootRelative
    alphaBoolBlend {blendOption=Linear}
    boundBone {boneName=dyn_torso_bow__r_04}
    componentLinearAccScale=<1 1 1>
    componentLinearVelScale=<1 1 1>
    physicsBodyDefinitions [{
      boundBone {boneName=dyn_torso_bow__r_04}
      boxExtents=<10 10 10>
      constraintSetup {angularLimitsMax=<0 5 20>, angularLimitsMin=<0 -20 -5>, angularTarget=<0 0 0>}
      localJointOffset=<0.5 0 0>
     }]
   }
   n15 = Anim Dynamics - Bone: dyn_torso_bow__r_05!(ComponentPose=@n14, Alpha=1, GravityOverride:Vector="0,0,0", ExternalForce="0,0,0") -> (Pose) {
    angularDampingOverride=0.9999
    angularSpringConstant=2000
    bOverrideAngularDamping=true
    bUseGravityOverride=true
    simulationSpace=RootRelative
    alphaBoolBlend {blendOption=Linear}
    boundBone {boneName=dyn_torso_bow__r_05}
    componentLinearAccScale=<1 1 1>
    componentLinearVelScale=<1 1 1>
    physicsBodyDefinitions [{
      boundBone {boneName=dyn_torso_bow__r_05}
      boxExtents=<10 10 10>
      constraintSetup {angularLimitsMax=<0 5 20>, angularLimitsMin=<0 -20 -10>, angularTarget=<0 0 0>}
      localJointOffset=<0.5 0 0>
     }]
   }
   n16 = Anim Dynamics - Bone: dyn_torso_bow__r_06!(ComponentPose=@n15, Alpha=1, GravityOverride:Vector="0,0,0", ExternalForce="0,0,0") -> (Pose) {
    angularDampingOverride=0.9999
    angularSpringConstant=2000
    bOverrideAngularDamping=true
    bUseGravityOverride=true
    simulationSpace=RootRelative
    alphaBoolBlend {blendOption=Linear}
    boundBone {boneName=dyn_torso_bow__r_06}
    componentLinearAccScale=<1 1 1>
    componentLinearVelScale=<1 1 1>
    physicsBodyDefinitions [{
      boundBone {boneName=dyn_torso_bow__r_06}
      boxExtents=<10 10 10>
      constraintSetup {angularLimitsMax=<0 5 20>, angularLimitsMin=<0 -20 -20>, angularTarget=<0 0 0>}
      localJointOffset=<0.5 0 0>
     }]
   }
   n17 = Anim Dynamics - Bone: dyn_foot_bow_l_01!(ComponentPose=@n16, Alpha=1, GravityOverride:Vector="0,0,0", ExternalForce="0,0,0") -> (Pose) {
    angularDampingOverride=0.9
    bOverrideAngularDamping=true
    bUseGravityOverride=true
    simulationSpace=RootRelative
    alphaBoolBlend {blendOption=Linear}
    boundBone {boneName=dyn_foot_bow_l_01}
    physicsBodyDefinitions [{
      boundBone {boneName=dyn_foot_bow_l_01}
      boxExtents=<10 10 10>
      constraintSetup {angularLimitsMax=<0 15 20>, angularLimitsMin=<0 -15 0>, angularTarget=<1 0 0>}
      localJointOffset=<0.5 0 0>
     }]
   }
   n18 = Anim Dynamics - Bone: dyn_foot_bow_l_02!(ComponentPose=@n17, Alpha=1, GravityOverride:Vector="0,0,0", ExternalForce="0,0,0") -> (Pose) {
    angularDampingOverride=0.9
    bOverrideAngularDamping=true
    bUseGravityOverride=true
    bUsePlanarLimit=false
    bUseSphericalLimits=true
    simulationSpace=RootRelative
    alphaBoolBlend {blendOption=Linear}
    boundBone {boneName=dyn_foot_bow_l_02}
    physicsBodyDefinitions [{
      collisionType=CustomSphere
      sphereCollisionRadius=2
      boundBone {boneName=dyn_foot_bow_l_02}
      boxExtents=<10 10 10>
      constraintSetup {angularLimitsMax=<0 15 20>, angularLimitsMin=<0 -15 0>, angularTarget=<1 0 0>}
      localJointOffset=<0.5 0 0>
     }]
    sphericalLimits [
     {limitRadius=10, drivingBone {boneName=foot_l}, sphereLocalOffset=<-2 5 0>}
     {limitRadius=7, drivingBone {boneName=foot_l}, sphereLocalOffset=<3 3 0>}
     {limitRadius=7, drivingBone {boneName=foot_l}, sphereLocalOffset=<-4.3 15 0>}
    ]
   }
   n19 = Anim Dynamics - Bone: dyn_foot_bow_l_03!(ComponentPose=@n18, Alpha=1, GravityOverride:Vector="0,0,0", ExternalForce="0,0,0") -> (Pose) {
    angularDampingOverride=0.9
    bOverrideAngularDamping=true
    bUseGravityOverride=true
    bUsePlanarLimit=false
    bUseSphericalLimits=true
    simulationSpace=RootRelative
    alphaBoolBlend {blendOption=Linear}
    boundBone {boneName=dyn_foot_bow_l_03}
    physicsBodyDefinitions [{
      collisionType=CustomSphere
      sphereCollisionRadius=2
      boundBone {boneName=dyn_foot_bow_l_03}
      boxExtents=<10 10 10>
      constraintSetup {angularLimitsMax=<0 15 20>, angularLimitsMin=<0 -15 0>, angularTarget=<1 0 0>}
      localJointOffset=<0.5 0 0>
     }]
    sphericalLimits [
     {limitRadius=10, drivingBone {boneName=foot_l}, sphereLocalOffset=<-2 5 0>}
     {limitRadius=7, drivingBone {boneName=foot_l}, sphereLocalOffset=<3 3 0>}
     {limitRadius=7, drivingBone {boneName=foot_l}, sphereLocalOffset=<-4.3 15 0>}
    ]
   }
   n20 = Anim Dynamics - Bone: dyn_foot_bow_l_04!(ComponentPose=@n19, Alpha=1, GravityOverride:Vector="0,0,0", ExternalForce="0,0,0") -> (Pose) {
    angularDampingOverride=0.9
    bOverrideAngularDamping=true
    bUseGravityOverride=true
    bUsePlanarLimit=false
    bUseSphericalLimits=true
    simulationSpace=RootRelative
    alphaBoolBlend {blendOption=Linear}
    boundBone {boneName=dyn_foot_bow_l_04}
    physicsBodyDefinitions [{
      collisionType=CustomSphere
      sphereCollisionRadius=2
      boundBone {boneName=dyn_foot_bow_l_04}
      boxExtents=<10 10 10>
      constraintSetup {angularLimitsMax=<0 15 20>, angularLimitsMin=<0 -15 0>, angularTarget=<1 0 0>}
      localJointOffset=<0.5 0 0>
     }]
    sphericalLimits [
     {limitRadius=10, drivingBone {boneName=foot_l}, sphereLocalOffset=<-2 5 0>}
     {limitRadius=7, drivingBone {boneName=foot_l}, sphereLocalOffset=<3 3 0>}
     {limitRadius=7, drivingBone {boneName=foot_l}, sphereLocalOffset=<-4.3 15 0>}
    ]
   }
   n21 = Anim Dynamics - Bone: dyn_foot_bow_l_08!(ComponentPose=@n20, Alpha=1, GravityOverride:Vector="0,0,0", ExternalForce="0,0,0") -> (Pose) {
    angularDampingOverride=0.9
    bOverrideAngularDamping=true
    bUseGravityOverride=true
    simulationSpace=RootRelative
    alphaBoolBlend {blendOption=Linear}
    boundBone {boneName=dyn_foot_bow_l_08}
    physicsBodyDefinitions [{
      boundBone {boneName=dyn_foot_bow_l_08}
      boxExtents=<10 10 10>
      constraintSetup {angularLimitsMax=<0 15 20>, angularLimitsMin=<0 -15 0>, angularTarget=<1 0 0>}
      localJointOffset=<0.5 0 0>
     }]
   }
   n22 = Anim Dynamics - Bone: dyn_foot_bow_l_05!(ComponentPose=@n21, Alpha=1, GravityOverride:Vector="0,0,0", ExternalForce="0,0,0") -> (Pose) {
    angularDampingOverride=0.9
    bOverrideAngularDamping=true
    bUseGravityOverride=true
    bUsePlanarLimit=false
    bUseSphericalLimits=true
    simulationSpace=RootRelative
    alphaBoolBlend {blendOption=Linear}
    boundBone {boneName=dyn_foot_bow_l_05}
    physicsBodyDefinitions [{
      collisionType=CustomSphere
      sphereCollisionRadius=2
      boundBone {boneName=dyn_foot_bow_l_05}
      boxExtents=<10 10 10>
      constraintSetup {angularLimitsMax=<0 15 20>, angularLimitsMin=<0 -15 0>, angularTarget=<1 0 0>}
      localJointOffset=<0.5 0 0>
     }]
    sphericalLimits [
     {limitRadius=10, drivingBone {boneName=foot_l}, sphereLocalOffset=<-2 5 0>}
     {limitRadius=7, drivingBone {boneName=foot_l}, sphereLocalOffset=<3 3 0>}
     {limitRadius=7, drivingBone {boneName=foot_l}, sphereLocalOffset=<-4.3 15 0>}
    ]
   }
   n23 = Anim Dynamics - Bone: dyn_foot_bow_l_06!(ComponentPose=@n22, Alpha=1, GravityOverride:Vector="0,0,0", ExternalForce="0,0,0") -> (Pose) {
    angularDampingOverride=0.9
    bOverrideAngularDamping=true
    bUseGravityOverride=true
    bUsePlanarLimit=false
    bUseSphericalLimits=true
    simulationSpace=RootRelative
    alphaBoolBlend {blendOption=Linear}
    boundBone {boneName=dyn_foot_bow_l_06}
    physicsBodyDefinitions [{
      collisionType=CustomSphere
      sphereCollisionRadius=2
      boundBone {boneName=dyn_foot_bow_l_06}
      boxExtents=<10 10 10>
      constraintSetup {angularLimitsMax=<0 15 20>, angularLimitsMin=<0 -15 0>, angularTarget=<1 0 0>}
      localJointOffset=<0.5 0 0>
     }]
    sphericalLimits [
     {limitRadius=10, drivingBone {boneName=foot_l}, sphereLocalOffset=<-2 5 0>}
     {limitRadius=7, drivingBone {boneName=foot_l}, sphereLocalOffset=<3 3 0>}
     {limitRadius=7, drivingBone {boneName=foot_l}, sphereLocalOffset=<-4.3 15 0>}
    ]
   }
   n24 = Anim Dynamics - Bone: dyn_foot_bow_l_07!(ComponentPose=@n23, Alpha=1, GravityOverride:Vector="0,0,0", ExternalForce="0,0,0") -> (Pose) {
    angularDampingOverride=0.9
    bOverrideAngularDamping=true
    bUseGravityOverride=true
    bUsePlanarLimit=false
    bUseSphericalLimits=true
    simulationSpace=RootRelative
    alphaBoolBlend {blendOption=Linear}
    boundBone {boneName=dyn_foot_bow_l_07}
    physicsBodyDefinitions [{
      collisionType=CustomSphere
      sphereCollisionRadius=2
      boundBone {boneName=dyn_foot_bow_l_07}
      boxExtents=<10 10 10>
      constraintSetup {angularLimitsMax=<0 15 20>, angularLimitsMin=<0 -15 0>, angularTarget=<1 0 0>}
      localJointOffset=<0.5 0 0>
     }]
    sphericalLimits [
     {limitRadius=10, drivingBone {boneName=foot_l}, sphereLocalOffset=<-2 5 0>}
     {limitRadius=7, drivingBone {boneName=foot_l}, sphereLocalOffset=<3 3 0>}
     {limitRadius=7, drivingBone {boneName=foot_l}, sphereLocalOffset=<-4.3 15 0>}
    ]
   }
   n25 = Anim Dynamics - Bone: dyn_foot_bow_r_01!(ComponentPose=@n24, Alpha=1, GravityOverride:Vector="0,0,0", ExternalForce="0,0,0") -> (Pose) {
    angularDampingOverride=0.9
    bOverrideAngularDamping=true
    bUseGravityOverride=true
    simulationSpace=RootRelative
    alphaBoolBlend {blendOption=Linear}
    boundBone {boneName=dyn_foot_bow_r_01}
    physicsBodyDefinitions [{
      boundBone {boneName=dyn_foot_bow_r_01}
      boxExtents=<10 10 10>
      constraintSetup {angularLimitsMax=<0 15 20>, angularLimitsMin=<0 -15 0>, angularTarget=<1 0 0>}
      localJointOffset=<0.5 0 0>
     }]
   }
   n26 = Anim Dynamics - Bone: dyn_foot_bow_r_02!(ComponentPose=@n25, Alpha=1, GravityOverride:Vector="0,0,0", ExternalForce="0,0,0") -> (Pose) {
    angularDampingOverride=0.9
    bOverrideAngularDamping=true
    bUseGravityOverride=true
    bUseSphericalLimits=true
    simulationSpace=RootRelative
    alphaBoolBlend {blendOption=Linear}
    boundBone {boneName=dyn_foot_bow_r_02}
    physicsBodyDefinitions [{
      collisionType=CustomSphere
      sphereCollisionRadius=2
      boundBone {boneName=dyn_foot_bow_r_02}
      boxExtents=<10 10 10>
      constraintSetup {angularLimitsMax=<0 15 20>, angularLimitsMin=<0 -15 0>, angularTarget=<1 0 0>}
      localJointOffset=<0.5 0 0>
     }]
    sphericalLimits [
     {limitRadius=10, drivingBone {boneName=foot_r}, sphereLocalOffset=<2 -5 0>}
     {limitRadius=7, drivingBone {boneName=foot_r}, sphereLocalOffset=<-3 -3 0>}
     {limitRadius=7, drivingBone {boneName=foot_r}, sphereLocalOffset=<4.3 -15 0>}
    ]
   }
   n27 = Anim Dynamics - Bone: dyn_foot_bow_r_03!(ComponentPose=@n26, Alpha=1, GravityOverride:Vector="0,0,0", ExternalForce="0,0,0") -> (Pose) {
    angularDampingOverride=0.9
    bOverrideAngularDamping=true
    bUseGravityOverride=true
    bUseSphericalLimits=true
    simulationSpace=RootRelative
    alphaBoolBlend {blendOption=Linear}
    boundBone {boneName=dyn_foot_bow_r_03}
    physicsBodyDefinitions [{
      collisionType=CustomSphere
      sphereCollisionRadius=2
      boundBone {boneName=dyn_foot_bow_r_03}
      boxExtents=<10 10 10>
      constraintSetup {angularLimitsMax=<0 15 20>, angularLimitsMin=<0 -15 0>, angularTarget=<1 0 0>}
      localJointOffset=<0.5 0 0>
     }]
    sphericalLimits [
     {limitRadius=10, drivingBone {boneName=foot_r}, sphereLocalOffset=<2 -5 0>}
     {limitRadius=7, drivingBone {boneName=foot_r}, sphereLocalOffset=<-3 -3 0>}
     {limitRadius=7, drivingBone {boneName=foot_r}, sphereLocalOffset=<4.3 -15 0>}
    ]
   }
   n28 = Anim Dynamics - Bone: dyn_foot_bow_r_04!(ComponentPose=@n27, Alpha=1, GravityOverride:Vector="0,0,0", ExternalForce="0,0,0") -> (Pose) {
    angularDampingOverride=0.9
    bOverrideAngularDamping=true
    bUseGravityOverride=true
    bUseSphericalLimits=true
    simulationSpace=RootRelative
    alphaBoolBlend {blendOption=Linear}
    boundBone {boneName=dyn_foot_bow_r_04}
    physicsBodyDefinitions [{
      collisionType=CustomSphere
      sphereCollisionRadius=2
      boundBone {boneName=dyn_foot_bow_r_04}
      boxExtents=<10 10 10>
      constraintSetup {angularLimitsMax=<0 15 20>, angularLimitsMin=<0 -15 0>, angularTarget=<1 0 0>}
      localJointOffset=<0.5 0 0>
     }]
    sphericalLimits [
     {limitRadius=10, drivingBone {boneName=foot_r}, sphereLocalOffset=<2 -5 0>}
     {limitRadius=7, drivingBone {boneName=foot_r}, sphereLocalOffset=<-3 -3 0>}
     {limitRadius=7, drivingBone {boneName=foot_r}, sphereLocalOffset=<4.3 -15 0>}
    ]
   }
   n29 = Anim Dynamics - Bone: dyn_foot_bow_r_08!(ComponentPose=@n28, Alpha=1, GravityOverride:Vector="0,0,0", ExternalForce="0,0,0") -> (Pose) {
    angularDampingOverride=0.9
    bOverrideAngularDamping=true
    bUseGravityOverride=true
    simulationSpace=RootRelative
    alphaBoolBlend {blendOption=Linear}
    boundBone {boneName=dyn_foot_bow_r_08}
    physicsBodyDefinitions [{
      boundBone {boneName=dyn_foot_bow_r_08}
      boxExtents=<10 10 10>
      constraintSetup {angularLimitsMax=<0 15 20>, angularLimitsMin=<0 -15 0>, angularTarget=<1 0 0>}
      localJointOffset=<0.5 0 0>
     }]
   }
   n30 = Anim Dynamics - Bone: dyn_foot_bow_r_05!(ComponentPose=@n29, Alpha=1, GravityOverride:Vector="0,0,0", ExternalForce="0,0,0") -> (Pose) {
    angularDampingOverride=0.9
    bOverrideAngularDamping=true
    bUseGravityOverride=true
    bUseSphericalLimits=true
    simulationSpace=RootRelative
    alphaBoolBlend {blendOption=Linear}
    boundBone {boneName=dyn_foot_bow_r_05}
    physicsBodyDefinitions [{
      collisionType=CustomSphere
      sphereCollisionRadius=1
      boundBone {boneName=dyn_foot_bow_r_05}
      boxExtents=<10 10 10>
      constraintSetup {angularLimitsMax=<0 15 20>, angularLimitsMin=<0 -15 0>, angularTarget=<1 0 0>}
      localJointOffset=<0.5 0 0>
     }]
    sphericalLimits [
     {limitRadius=10, drivingBone {boneName=foot_r}, sphereLocalOffset=<2 -5 0>}
     {limitRadius=7, drivingBone {boneName=foot_r}, sphereLocalOffset=<-3 -3 0>}
     {limitRadius=7, drivingBone {boneName=foot_r}, sphereLocalOffset=<4.3 -15 0>}
    ]
   }
   n31 = Anim Dynamics - Bone: dyn_foot_bow_r_06!(ComponentPose=@n30, Alpha=1, GravityOverride:Vector="0,0,0", ExternalForce="0,0,0") -> (Pose) {
    angularDampingOverride=0.9
    bOverrideAngularDamping=true
    bUseGravityOverride=true
    bUseSphericalLimits=true
    simulationSpace=RootRelative
    alphaBoolBlend {blendOption=Linear}
    boundBone {boneName=dyn_foot_bow_r_06}
    physicsBodyDefinitions [{
      collisionType=CustomSphere
      sphereCollisionRadius=1
      boundBone {boneName=dyn_foot_bow_r_06}
      boxExtents=<10 10 10>
      constraintSetup {angularLimitsMax=<0 15 20>, angularLimitsMin=<0 -15 0>, angularTarget=<1 0 0>}
      localJointOffset=<0.5 0 0>
     }]
    sphericalLimits [
     {limitRadius=10, drivingBone {boneName=foot_r}, sphereLocalOffset=<2 -5 0>}
     {limitRadius=7, drivingBone {boneName=foot_r}, sphereLocalOffset=<-3 -3 0>}
     {limitRadius=7, drivingBone {boneName=foot_r}, sphereLocalOffset=<4.3 -15 0>}
    ]
   }
   n32 = Anim Dynamics - Bone: dyn_foot_bow_r_07!(ComponentPose=@n31, Alpha=1, GravityOverride:Vector="0,0,0", ExternalForce="0,0,0") -> (Pose) {
    angularDampingOverride=0.9
    bOverrideAngularDamping=true
    bUseGravityOverride=true
    bUseSphericalLimits=true
    simulationSpace=RootRelative
    alphaBoolBlend {blendOption=Linear}
    boundBone {boneName=dyn_foot_bow_r_07}
    physicsBodyDefinitions [{
      collisionType=CustomSphere
      sphereCollisionRadius=1
      boundBone {boneName=dyn_foot_bow_r_07}
      boxExtents=<10 10 10>
      constraintSetup {angularLimitsMax=<0 15 20>, angularLimitsMin=<0 -15 0>, angularTarget=<1 0 0>}
      localJointOffset=<0.5 0 0>
     }]
    sphericalLimits [
     {limitRadius=10, drivingBone {boneName=foot_r}, sphereLocalOffset=<2 -5 0>}
     {limitRadius=7, drivingBone {boneName=foot_r}, sphereLocalOffset=<-3 -3 0>}
     {limitRadius=7, drivingBone {boneName=foot_r}, sphereLocalOffset=<4.3 -15 0>}
    ]
   }
   n33 = Anim Dynamics - Bone: Belt_01!(ComponentPose=@n32, Alpha=1, GravityOverride:Vector="0,0,0", ExternalForce="0,0,0") -> (Pose) {
    angularDampingOverride=0.9999
    angularSpringConstant=2000
    bOverrideAngularDamping=true
    bUseGravityOverride=true
    simulationSpace=RootRelative
    alphaBoolBlend {blendOption=Linear}
    boundBone {boneName=Belt_01}
    componentLinearAccScale=<1 1 1>
    componentLinearVelScale=<1 1 1>
    physicsBodyDefinitions [{
      collisionType=OuterSphere
      sphereCollisionRadius=1
      boundBone {boneName=Belt_01}
      boxExtents=<10 10 10>
      constraintSetup {angularLimitsMax=<15 15 20>, angularLimitsMin=<-15 -15 -30>, angularTarget=<0 0 0>}
      localJointOffset=<0.5 0 0>
     }]
   }
   n34 = Anim Dynamics - Bone: Belt_02!(ComponentPose=@n33, Alpha=1, GravityOverride:Vector="0,0,0", ExternalForce="0,0,0") -> (Pose) {
    angularDampingOverride=0.9999
    angularSpringConstant=2000
    bOverrideAngularDamping=true
    bUseGravityOverride=true
    simulationSpace=RootRelative
    alphaBoolBlend {blendOption=Linear}
    boundBone {boneName=Belt_02}
    componentLinearAccScale=<1 1 1>
    componentLinearVelScale=<1 1 1>
    physicsBodyDefinitions [{
      collisionType=OuterSphere
      sphereCollisionRadius=1
      boundBone {boneName=Belt_02}
      boxExtents=<10 10 10>
      constraintSetup {angularLimitsMax=<1 15 10>, angularLimitsMin=<-1 -20 -10>, angularTarget=<0 0 0>}
      localJointOffset=<0.7 0 0>
     }]
    planarLimits [{drivingBone {boneName=dyn_skirt_front_c_Base}, planeTransform {location=<9.1502 -4.5668 7.6816>, rotation=<rot 2.9314 -1.0448 81.1441>}}]
   }
   n35 = Anim Dynamics - Bone: Belt_03!(ComponentPose=@n34, Alpha=1, GravityOverride:Vector="0,0,0", ExternalForce="0,0,0") -> (Pose) {
    angularDampingOverride=0.9999
    angularSpringConstant=2000
    bOverrideAngularDamping=true
    bUseGravityOverride=true
    simulationSpace=RootRelative
    alphaBoolBlend {blendOption=Linear}
    boundBone {boneName=Belt_03}
    componentLinearAccScale=<1 1 1>
    componentLinearVelScale=<1 1 1>
    physicsBodyDefinitions [{
      collisionType=OuterSphere
      sphereCollisionRadius=1
      boundBone {boneName=Belt_03}
      boxExtents=<10 10 10>
      constraintSetup {angularLimitsMax=<5 1 5>, angularLimitsMin=<-5 -1 -5>, angularTarget=<0 0 0>}
      localJointOffset=<1 0 0>
     }]
    planarLimits [{drivingBone {boneName=dyn_skirt_front_c_Base}, planeTransform {location=<-6.7897 1.4926 3.3449>, rotation=<rot -3.4515 -11.5898 83.6657>}}]
   }
   Output Pose(Result=Component To Local(ComponentPose=@n35)) {layerGroup=None}
   n36 = Control Rig!(
    Source=Copy Pose From Mesh() {bUseAttachedParent=true}
    Alpha
    is Emote:bool=Get bIsPlayingEmote()
    is DBNO:bool=Get bIsDBNO()
    IsAlfaEMOTE:double=Get IsAlfaEMOTE() -> (IsAlfaEMOTE:double)
    isFE:bool=Get IsFrEnd() -> (IsFrEnd:bool)
   ) -> (Pose) {
    alphaBoolBlend {blendOption=Linear}
    controlRigAssetReference {blueprintRigClass=&/F_MED_ElegantLilyAnkle_CtrlRig.F_MED_ElegantLilyAnkle_CtrlRig_C}
   }
   Save cached pose 'PreSim'(Pose=@n36) {cacheName=PreSim}
   n37 = Anim Dynamics - Bone: dyn_skirt_side_l_02!(ComponentPose=Local To Component(LocalPose=Use cached pose 'PreSim'()), Alpha=1, GravityOverride:Vector="0,0,0", ExternalForce="0,0,0") -> (Pose) {
    angularDampingOverride=0.9999
    angularSpringConstant=2000
    bOverrideAngularDamping=true
    bUseGravityOverride=true
    simulationSpace=RootRelative
    alphaBoolBlend {blendOption=Linear}
    boundBone {boneName=dyn_skirt_side_l_02}
    componentLinearAccScale=<1 1 1>
    componentLinearVelScale=<1 1 1>
    physicsBodyDefinitions [{
      boundBone {boneName=dyn_skirt_side_l_02}
      boxExtents=<10 10 10>
      constraintSetup {angularLimitsMax=<0 1 12>, angularLimitsMin=<0 -1 -12>, angularTarget=<0 0 0>}
      localJointOffset=<0.5 0 0>
     }]
   }
   n38 = Anim Dynamics - Bone: dyn_skirt_back_l_02!(ComponentPose=@n37, Alpha=1, GravityOverride:Vector="0,0,0", ExternalForce="0,0,0") -> (Pose) {
    angularDampingOverride=0.9999
    angularSpringConstant=2000
    bOverrideAngularDamping=true
    bUseGravityOverride=true
    simulationSpace=RootRelative
    alphaBoolBlend {blendOption=Linear}
    boundBone {boneName=dyn_skirt_back_l_02}
    componentLinearAccScale=<1 1 1>
    componentLinearVelScale=<1 1 1>
    physicsBodyDefinitions [{
      boundBone {boneName=dyn_skirt_back_l_02}
      boxExtents=<10 10 10>
      constraintSetup {angularLimitsMax=<0 1 12>, angularLimitsMin=<0 -1 -12>, angularTarget=<0 0 0>}
      localJointOffset=<0.5 0 0>
     }]
   }
   n39 = Anim Dynamics - Bone: dyn_skirt_back_c_02!(ComponentPose=@n38, Alpha=1, GravityOverride:Vector="0,0,0", ExternalForce="0,0,0") -> (Pose) {
    angularDampingOverride=0.9999
    angularSpringConstant=2000
    bOverrideAngularDamping=true
    bUseGravityOverride=true
    simulationSpace=RootRelative
    alphaBoolBlend {blendOption=Linear}
    boundBone {boneName=dyn_skirt_back_c_02}
    componentLinearAccScale=<1 1 1>
    componentLinearVelScale=<1 1 1>
    physicsBodyDefinitions [{
      boundBone {boneName=dyn_skirt_back_c_02}
      boxExtents=<10 10 10>
      constraintSetup {angularLimitsMax=<0 1 20>, angularLimitsMin=<0 -1 -12>, angularTarget=<0 0 0>}
      localJointOffset=<0.5 0 0>
     }]
   }
   n40 = Anim Dynamics - Bone: dyn_skirt_back_r_02!(ComponentPose=@n39, Alpha=1, GravityOverride:Vector="0,0,0", ExternalForce="0,0,0") -> (Pose) {
    angularDampingOverride=0.9999
    angularSpringConstant=2000
    bOverrideAngularDamping=true
    bUseGravityOverride=true
    simulationSpace=RootRelative
    alphaBoolBlend {blendOption=Linear}
    boundBone {boneName=dyn_skirt_back_r_02}
    componentLinearAccScale=<1 1 1>
    componentLinearVelScale=<1 1 1>
    physicsBodyDefinitions [{
      boundBone {boneName=dyn_skirt_back_r_02}
      boxExtents=<10 10 10>
      constraintSetup {angularLimitsMax=<0 1 12>, angularLimitsMin=<0 -1 -12>, angularTarget=<0 0 0>}
      localJointOffset=<0.5 0 0>
     }]
   }
   n41 = Anim Dynamics - Bone: dyn_skirt_side_r_02!(ComponentPose=@n40, Alpha=1, GravityOverride:Vector="0,0,0", ExternalForce="0,0,0") -> (Pose) {
    angularDampingOverride=0.9999
    angularSpringConstant=2000
    bOverrideAngularDamping=true
    bUseGravityOverride=true
    simulationSpace=RootRelative
    alphaBoolBlend {blendOption=Linear}
    boundBone {boneName=dyn_skirt_side_r_02}
    componentLinearAccScale=<1 1 1>
    componentLinearVelScale=<1 1 1>
    physicsBodyDefinitions [{
      boundBone {boneName=dyn_skirt_side_r_02}
      boxExtents=<10 10 10>
      constraintSetup {angularLimitsMax=<0 1 12>, angularLimitsMin=<0 -1 -12>, angularTarget=<0 0 0>}
      localJointOffset=<0.5 0 0>
     }]
   }
   n42 = Control Rig!(
    Source=Component To Local(ComponentPose=@n41)
    Alpha
    isJumping:bool=Get isJumping() -> (isJumping:bool)
    Y movement:double=Get Y Movement() -> (Y Movement:double)
    X movement:double=Get X Movement() -> (X Movement:double)
    is Emote=Get bIsPlayingEmote()
    is AltSkin:bool=Get Is AltSkin() -> (Is AltSkin:bool)
    IsAlfaEMOTE=Get IsAlfaEMOTE()
    isFrEnd:bool=Get IsFrEnd()
   ) -> (Pose) {
    alphaBoolBlend {blendOption=Linear}
    controlRigAssetReference {blueprintRigClass=&/F_MED_ElegantLilyAnkle_Strap_CtrlRig.F_MED_ElegantLilyAnkle_Strap_CtrlRig_C}
   }
   Save cached pose 'Cash pose'(Pose=@n42) {cacheName=Cash pose}
   n43 = RigidBody!(ComponentPose=Local To Component(LocalPose=Use cached pose 'Cash pose'()), Alpha=0.4, ExternalForce) -> (Pose) {
    overridePhysicsAsset=&/F_MED_ElegantLilyAnkle_strap_FRONTE
    alphaBoolBlend {blendOption=Linear}
   }
   n44 = RigidBody!(ComponentPose=Local To Component(LocalPose=Use cached pose 'Cash pose'()), Alpha=0.7, OverrideWorldGravity:Vector="0,0,0", ExternalForce) -> (Pose) {
    bOverrideWorldGravity=true
    overridePhysicsAsset=&/F_MED_ElegantLilyAnkle_strap
    simulationSpace=BaseBoneSpace
    alphaBoolBlend {blendOption=Linear}
    baseBoneRef {boneName=pelvis}
   }
   Save cached pose 'RigidBody'(Pose=Blend Poses by bool(BlendPose_0=Component To Local(ComponentPose=@n43), BlendPose_1=Component To Local(ComponentPose=@n44)) {blendType=Linear}) {cacheName=RigidBody}
  ]}
 EventGraph {nodes [
   Event BlueprintUpdateAnimation()
   > n1 = Gravity Override State Function DEPRECATED(CharacterPartAnimInstance=Self-Reference(), GameTimeInSeconds=GetGameTimeInSeconds(), GravityOverrideParams=Get Gravity Override Params())
   > Set Gravity Override(Gravity Override=@n1)
   > Set isJumping(isJumping:bool=OR Boolean(A=Get bIsJumping(), B=Get bIsFalling())) -> (Output_Get:bool)
   > Set Is AltSkin(Is AltSkin:bool=NOT Boolean(A=Get bRigidBodyEnabled())) -> (Output_Get:bool)
   n2 = Get bIsPlayingEmote()
   > Set IsFrEnd(IsFrEnd:bool=AND Boolean(A=NOT Boolean(A=@n2), B=Get bIsInFrontEnd())) -> (Output_Get:bool)
   > n3 = PlayerMovement(AnimInstance=Self-Reference(), XDotProductCoefficient=1, YDotProductCoefficient=1, ZDotProductCoefficient=1)
   > Set Y Movement(Y Movement:double=@n3.YMovement) -> (Output_Get:double)
   > Set X Movement(X Movement:double=@n3) -> (Output_Get:double)
   n4 = GetAnimInstance(self=GetOwnerSkeletalMeshComponent())
   > n5 = Is Valid(InputObject=@n4)
   Set IsAlfaEMOTE(execute=@n5.Is Not Valid, IsAlfaEMOTE:double=0) -> (Output_Get:double)
   Set IsAlfaEMOTE(
    execute=@n5.Is Valid
    IsAlfaEMOTE=SelectFloat(
     A=1
     B=0
     bPickA=Montage_IsPlaying(self=@n4, Montage=/BRCosmetics/Animation/Game/MainPlayer/Emotes/ElegantLilyCharm/CMF/Emote_ElegantLilyCharm_CMF_M)
    )
   )
   AND Boolean(A="integer >= integer"(A=Get CurrentLOD(), B=1), B=false)
  ]}
]
variables [
 Context {type=AnimExecutionContext, default {}}
 Gravity Override {type=Vector}
 Gravity Override Params {type=GravityOverrideParamsStruct}
 Is AltSkin {type=boolean}
 IsAlfaEMOTE {type=number}
 IsFrEnd {type=boolean}
 isJumping {type=boolean}
 X Movement {type=number}
 Y Movement {type=number}
]
