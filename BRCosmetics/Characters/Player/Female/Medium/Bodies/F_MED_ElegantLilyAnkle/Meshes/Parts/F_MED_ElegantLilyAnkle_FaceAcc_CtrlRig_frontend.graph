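# Control Rig for the ElegantLilyAnkle female body's face accessory part that poses the left and right front hair bones, apparently for the frontend.
# Weighted parent constraints drive each side's base and lower chain bones from spine, clavicle and arm bones, then a ModifyTransforms step per side adjusts the chain.
# The IsFE bool, converted to a float, weights every step, so the rig has no effect at its default of false.
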ControlRigBlueprint /BRCosmetics/Characters/Player/Female/Medium/Bodies/F_MED_ElegantLilyAnkle/Meshes/Parts/F_MED_ElegantLilyAnkle_FaceAcc_CtrlRig_frontend
graphs [
 RigVMFunctionLibrary {links [], nodes []}
 RigVMModel {nodes [
   RigUnit_BeginExecution = &RigUnit_BeginExecution()
   n1 = &RigVMFunction_MathBoolToFloat(Value=RigVMVariableNode() {variable=IsFE}.Value)
   &RigUnit_ParentConstraint(
    AdvancedSettings=(InterpolationType=Average,RotationOrderForFilter=XZY)
    Child=(Type=Bone,Name="Base_hair_Front_L")
    Parents="((Item=(Type=Bone,Name=\"upperarm_twist_01_l\"),Weight=0.0195),(Item=(Type=Bone,Name=\"spine_04\"),Weight=0.0479),(Item=(Type=Bone,Name=\"spine_05\"),Weight=0.4686),(Item=(Type=Bone,Name=\"clavicle_l\"),Weight=0.4638),(Item=(Type=Bone,Name=\"clavicle_r\"),Weight=0), ...(truncated)"
    ExecutePin=@RigUnit_BeginExecution.ExecutePin
    Weight=@n1.Result
   )
   > &RigUnit_ParentConstraint(
    AdvancedSettings=(InterpolationType=Average,RotationOrderForFilter=XZY)
    Child=(Type=Bone,Name="Dyn_hair_Front_L_02")
    Parents="((Item=(Type=Bone,Name=\"upperarm_twist_01_l\"),Weight=0.0463),(Item=(Type=Bone,Name=\"spine_04\"),Weight=0.3772),(Item=(Type=Bone,Name=\"spine_05\"),Weight=0.3752),(Item=(Type=Bone,Name=\"clavicle_l\"),Weight=0.2013),(Item=(Type=Bone,Name=\"clavicle_r\"),Weight=0), ...(truncated)"
    Weight=@n1.Result
   )
   > &RigUnit_ParentConstraint(
    AdvancedSettings=(InterpolationType=Average,RotationOrderForFilter=XZY)
    Child=(Type=Bone,Name="Dyn_hair_Front_L_03")
    Parents="((Item=(Type=Bone,Name=\"upperarm_twist_01_l\"),Weight=0.0113),(Item=(Type=Bone,Name=\"spine_04\"),Weight=0.7242),(Item=(Type=Bone,Name=\"spine_05\"),Weight=0.2357),(Item=(Type=Bone,Name=\"clavicle_l\"),Weight=0.0287),(Item=(Type=Bone,Name=\"upperarm_l\"),Weight=0), ...(truncated)"
    Weight=@n1.Result
   )
   > &RigUnit_ParentConstraint(
    AdvancedSettings=(InterpolationType=Average,RotationOrderForFilter=XZY)
    Child=(Type=Bone,Name="Dyn_hair_Front_L_04")
    Parents="((Item=(Type=Bone,Name=\"upperarm_twist_01_l\"),Weight=0.0113),(Item=(Type=Bone,Name=\"spine_04\"),Weight=0.7242),(Item=(Type=Bone,Name=\"spine_05\"),Weight=0.2357),(Item=(Type=Bone,Name=\"clavicle_l\"),Weight=0.0287),(Item=(Type=Bone,Name=\"upperarm_l\"),Weight=0), ...(truncated)"
    Weight=@n1.Result
   )
   > &RigUnit_ModifyTransforms(
    ItemToModify="((Item=(Type=Bone,Name=\"Dyn_hair_Front_L_01\"),Transform=(Rotation=(X=-0.0634,Y=0.0209,Z=0.0902,W=0.9937),Translation=(X=0.0004,Y=-0.0004,Z=3.0179),Scale3D=(X=1,Y=1,Z=1))),(Item=(Type=Bone,Name=\"Dyn_hair_Front_L_02\"),Transform=(Rotation=(X=0.0604,Y=-0.188,Z ...(truncated)"
    WeightMinimum=0
    Weight=@n1.Result
   )
   > &RigUnit_ParentConstraint(
    AdvancedSettings=(InterpolationType=Average,RotationOrderForFilter=XZY)
    Child=(Type=Bone,Name="Base_hair_Front_R")
    Parents="((Item=(Type=Bone,Name=\"upperarm_twist_01_r\"),Weight=0.0083),(Item=(Type=Bone,Name=\"spine_04\"),Weight=0.06),(Item=(Type=Bone,Name=\"spine_05\"),Weight=0.7745),(Item=(Type=Bone,Name=\"clavicle_l\"),Weight=0),(Item=(Type=Bone,Name=\"clavicle_r\"),Weight=0.157),(It ...(truncated)"
    Weight=@n1.Result
   )
   > &RigUnit_ParentConstraint(
    AdvancedSettings=(InterpolationType=Average,RotationOrderForFilter=XZY)
    Child=(Type=Bone,Name="Dyn_hair_Front_R_02")
    Parents="((Item=(Type=Bone,Name=\"upperarm_twist_01_r\"),Weight=0.0463),(Item=(Type=Bone,Name=\"spine_04\"),Weight=0.3772),(Item=(Type=Bone,Name=\"spine_05\"),Weight=0.3752),(Item=(Type=Bone,Name=\"clavicle_l\"),Weight=0),(Item=(Type=Bone,Name=\"clavicle_r\"),Weight=0.2013), ...(truncated)"
    Weight=@n1.Result
   )
   > &RigUnit_ParentConstraint(
    AdvancedSettings=(InterpolationType=Average,RotationOrderForFilter=XZY)
    Child=(Type=Bone,Name="Dyn_hair_Front_R_03")
    Parents="((Item=(Type=Bone,Name=\"upperarm_twist_01_l\"),Weight=0.0088),(Item=(Type=Bone,Name=\"spine_04\"),Weight=0.842),(Item=(Type=Bone,Name=\"spine_05\"),Weight=0.1286),(Item=(Type=Bone,Name=\"clavicle_r\"),Weight=0.0205),(Item=(Type=Bone,Name=\"upperarm_r\"),Weight=0),( ...(truncated)"
    Weight=@n1.Result
   )
   > &RigUnit_ParentConstraint(
    AdvancedSettings=(InterpolationType=Average,RotationOrderForFilter=XZY)
    Child=(Type=Bone,Name="Dyn_hair_Front_R_04")
    Parents=((Item=(Type=Bone,Name="lowerarm_twist_02_r"),Weight=0),(Item=(Type=Bone,Name="spine_04"),Weight=0.9884),(Item=(Type=Bone,Name="spine_05"),Weight=0.0116))
    Weight=@n1.Result
   )
   > &RigUnit_ModifyTransforms(
    ItemToModify="((Item=(Type=Bone,Name=\"Dyn_hair_Front_R_01\"),Transform=(Rotation=(X=0,Y=-0.0358,Z=0,W=-0.9994),Translation=(X=0,Y=0,Z=0),Scale3D=(X=1,Y=1,Z=1))),(Item=(Type=Bone,Name=\"Dyn_hair_Front_R_02\"),Transform=(Rotation=(X=0,Y=0.0254,Z=0,W=-0.9997),Translation=(X=0 ...(truncated)"
    WeightMinimum=0
    Weight=@n1.Result
   )
  ]}
]
hierarchy [
 root {loc=<0 0 0>, rot=<0 0 0 1>}
 pelvis {parent=root, loc=<0 -0.5301 98.0571>, rot=<0.0127 0.707 -0.0127 -0.707>}
 spine_01 {parent=pelvis, loc=<6.1208 -0.9328 0>, rot=<0 0 0.034 -0.9994>}
 spine_02 {parent=spine_01, loc=<6.5996 -0.2095 0>, rot=<0 0 0.0238 -0.9997>}
 spine_03 {parent=spine_02, loc=<6.4914 -0.1237 0>, rot=<0 0 0.0149 -0.9999>}
 spine_04 {parent=spine_03, loc=<6.9598 0.0794 0>, rot=<0 0 -0.0132 -0.9999>}
 spine_05 {parent=spine_04, loc=<6.9987 0 0>, rot=<0 0 -0.0015 -1>}
 clavicle_l {parent=spine_05, loc=<5.4831 -0.1053 -3.2946>, rot=<0.0465 -0.7289 -0.009 -0.6829>}
 upperarm_l {parent=clavicle_l, loc=<9.8658 -1.2136 -0.4105>, rot=<0.0164 -0.3245 -0.0269 -0.9454>}
 lowerarm_l {parent=upperarm_l, loc=<26.465 -0.8345 0.259>, rot=<0.0036 0.0161 -0.2298 -0.9731>}
 hairClavicleCorr_l {parent=clavicle_l, loc=<6.0994 -1.2432 4.5671>, rot=<-0.0012 0.0335 0.0362 -0.9988>}
 neck_01 {parent=spine_05, loc=<11.3138 0.1529 0>, rot=<0 0 -0.2003 -0.9797>}
 neck_02 {parent=neck_01, loc=<4.5627 0 0>, rot=<0 0 0.0504 -0.9987>}
 head {parent=neck_02, loc=<5.3914 0 0>, rot=<0 0 0.115 -0.9934>}
 Dyn_curl_R_01 {parent=head, loc=<10.5684 8.2276 0.0924>, rot=<0.5945 0.2685 0.755 0.0666>}
 Dyn_curl_R_02 {parent=Dyn_curl_R_01, loc=<2.2899 -0.0931 -1.1628>, rot=<-0.1487 -0.4717 0.2633 -0.8283>}
 Dyn_curl_R_03 {parent=Dyn_curl_R_02, loc=<2.5975 -0.0363 -0.1496>, rot=<-0.0111 0.0439 -0.3897 -0.9198>}
 Dyn_curl_L_01 {parent=head, loc=<4.014 4.7316 -5.9741>, rot=<0.3912 -0.8838 -0.2432 -0.082>}
 Dyn_curl_L_02 {parent=Dyn_curl_L_01, loc=<2.3695 0.3756 0.4219>, rot=<-0.1478 0.2134 -0.5498 -0.794>}
 Dyn_Headphones_Pillow_l_Base {parent=head, loc=<3.8362 -1.0616 -9.6379>, rot=<0.0029 -0.7071 -0.0029 -0.7071>}
 Dyn_Headphones_Pillow_l_01 {parent=Dyn_Headphones_Pillow_l_Base, loc=<0 -2.084 1.4081>, rot=<-0.3536 0.6124 0.3536 -0.6124>}
 Dyn_Headphones_Pillow_l_02 {parent=Dyn_Headphones_Pillow_l_Base, loc=<0 -2.3113 -1.7553>, rot=<0.2706 -0.6533 0.2706 -0.6533>}
 Dyn_Headphones_Pillow_l_03 {parent=Dyn_Headphones_Pillow_l_Base, loc=<0 0.9928 -2.415>, rot=<-0.0923 -0.7011 -0.0923 -0.7011>}
 Dyn_Headphones_Pillow_l_04 {parent=Dyn_Headphones_Pillow_l_Base, loc=<0 2.309 0>, rot=<-0.0923 0.0923 -0.7011 -0.7011>}
 Dyn_Headphones_Pillow_l_05 {parent=Dyn_Headphones_Pillow_l_Base, loc=<0 0.4511 2.6469>, rot=<0.0923 0.7011 -0.0923 -0.7011>}
 Dyn_Headphones_Pillow_r_Base {parent=head, loc=<3.8357 -1.0616 9.6379>, rot=<0.7071 0.0029 -0.7071 0.0029>}
 Dyn_Headphones_Pillow_r_01 {parent=Dyn_Headphones_Pillow_r_Base, loc=<0 2.084 -1.409>, rot=<-0.3536 0.6124 0.3536 -0.6124>}
 Dyn_Headphones_Pillow_r_02 {parent=Dyn_Headphones_Pillow_r_Base, loc=<0 2.3113 1.755>, rot=<0.2706 -0.6533 0.2706 -0.6533>}
 Dyn_Headphones_Pillow_r_03 {parent=Dyn_Headphones_Pillow_r_Base, loc=<0 -0.9928 2.415>, rot=<-0.0923 -0.7011 -0.0923 -0.7011>}
 Dyn_Headphones_Pillow_r_04 {parent=Dyn_Headphones_Pillow_r_Base, loc=<0 -2.309 0>, rot=<-0.0923 0.0923 -0.7011 -0.7011>}
 Dyn_Headphones_Pillow_r_05 {parent=Dyn_Headphones_Pillow_r_Base, loc=<0 -0.4511 -2.647>, rot=<0.0923 0.7011 -0.0923 -0.7011>}
 Dyn_Earring {parent=head, loc=<-0.2316 0.0392 -6.1975>, rot=<0.0041 -0.9914 0.1305 -0.0005>}
 hairCorrBack {parent=head, loc=<-0.0479 -6.939 0>, rot=<0.1819 -0.6833 0.6833 -0.1819>}
 hair_corr {parent=head, loc=<-1.9705 -3.7901 9.6411>, rot=<0.0029 -0.7071 -0.0029 -0.7071>}
 hair_corr_02 {parent=head, loc=<-3.1187 5.7726 7.2403>, rot=<0.0029 -0.7071 -0.0029 -0.7071>}
 Head_corr_03 {parent=head, loc=<0 0 0>, rot=<0.0029 -0.7071 -0.0029 -0.7071>}
 clavicle_r {parent=spine_05, loc=<5.4831 -0.1053 3.2946>, rot=<0.7289 0.0465 -0.6829 0.009>}
 upperarm_r {parent=clavicle_r, loc=<-9.8658 1.2136 0.4105>, rot=<0.0164 -0.3245 -0.0269 -0.9454>}
 lowerarm_r {parent=upperarm_r, loc=<-26.465 0.8345 -0.259>, rot=<0.0036 0.0161 -0.2298 -0.9731>}
 hairClavicleCorr_r {parent=clavicle_r, loc=<-6.9233 0.8691 -4.9186>, rot=<0.9976 -0.0668 0.0195 -0.0016>}
 Base_hair_Front_R {parent=spine_05, loc=<9.6689 5.1325 7.6689>, rot=<0.1943 0.6248 0.7403 0.1547>}
 Dyn_hair_Front_R_01 {parent=Base_hair_Front_R, loc=<0 0 0>, rot=<0 0 0 1>}
 Dyn_hair_Front_R_02 {parent=Dyn_hair_Front_R_01, loc=<6.9076 0 0>, rot=<0.0942 0.0033 -0.0622 -0.9936>}
 Dyn_hair_Front_R_03 {parent=Dyn_hair_Front_R_02, loc=<6.4541 0.3801 0.0004>, rot=<0.0044 -0.0886 0.232 -0.9687>}
 Dyn_hair_Front_R_04 {parent=Dyn_hair_Front_R_03, loc=<5.4086 0 0>, rot=<0.0107 -0.2203 -0.2055 -0.9535>}
 Dyn_hair_Front_R_05 {parent=Dyn_hair_Front_R_03, loc=<4.5961 -1.4979 0.3246>, rot=<0.0713 -0.1 0.0303 -0.992>}
 Base_hair_Front_L {parent=spine_05, loc=<10.5877 2.5481 -7.8042>, rot=<-0.2049 -0.6854 -0.6655 -0.2129>}
 Dyn_hair_Front_L_01 {parent=Base_hair_Front_L, loc=<0 0 0>, rot=<-0.0003 0.0028 0.0007 -1>}
 Dyn_hair_Front_L_02 {parent=Dyn_hair_Front_L_01, loc=<7.0277 0 0>, rot=<-0.0497 0.0371 0.0437 -0.9971>}
 Dyn_hair_Front_L_03 {parent=Dyn_hair_Front_L_02, loc=<6.6657 0 0>, rot=<-0.0082 -0.1555 0.0091 -0.9878>}
 Dyn_hair_Front_L_04 {parent=Dyn_hair_Front_L_03, loc=<5.7002 0 0>, rot=<0.0354 0.0022 -0.0835 -0.9959>}
 Dyn_hair_Front_L_05 {parent=Dyn_hair_Front_L_03, loc=<0.1084 1.3078 0.0273>, rot=<-0.005 0.1243 -0.275 -0.9534>}
 Base_hair_back_L {parent=spine_05, loc=<12.9411 -6.6943 0>, rot=<0.2064 -0.6763 0.6763 -0.2064>}
 Dyn_hair_back_L_01 {parent=Base_hair_back_L, loc=<0 0 0>, rot=<0 0 0 1>}
 Dyn_hair_back_L_02 {parent=Dyn_hair_back_L_01, loc=<7.2475 0 0>, rot=<0 -0.2705 0 -0.9627>}
 Dyn_hair_back_L_03 {parent=Dyn_hair_back_L_02, loc=<7.42 0 0>, rot=<0 0.0028 0 -1>}
 Dyn_hair_back_L_04 {parent=Dyn_hair_back_L_03, loc=<7.1032 0 0>, rot=<0 -0.0305 0 -0.9995>}
 Dyn_hair_back_L_12 {parent=Dyn_hair_back_L_04, loc=<1.4918 -9.6525 -1.4633>, rot=<0 -0.2283 0 -0.9736>}
 Dyn_hair_back_L_14 {parent=Dyn_hair_back_L_12, loc=<5.9275 1.1029 0.3188>, rot=<-0.1175 0.4067 0.0157 -0.9058>}
 Dyn_hair_back_L_13 {parent=Dyn_hair_back_L_04, loc=<0.7925 -10.4317 0.8789>, rot=<-0.0193 -0.1454 0.2869 -0.9467>}
 Dyn_hair_back_L_07 {parent=Dyn_hair_back_L_04, loc=<8.4829 5.6888 -0.8046>, rot=<0.1915 0.2118 -0.1717 -0.9429>}
 Dyn_hair_back_L_08 {parent=Dyn_hair_back_L_04, loc=<9.4681 3.3934 -0.0625>, rot=<0.0572 0.1711 -0.0106 -0.9835>}
 Dyn_hair_back_L_09 {parent=Dyn_hair_back_L_04, loc=<9.8781 0.7144 1.482>, rot=<-0.0141 0.0145 0.1036 -0.9944>}
 Dyn_hair_back_L_10 {parent=Dyn_hair_back_L_04, loc=<8.3999 -1.9085 -1.2085>, rot=<-0.1258 0.1645 0.0425 -0.9774>}
 Dyn_hair_back_L_11 {parent=Dyn_hair_back_L_04, loc=<6.8319 -4.8132 -0.0603>, rot=<-0.0107 -0.2012 -0.0518 -0.9781>}
 Dyn_hair_back_L_06 {parent=Dyn_hair_back_L_02, loc=<7.6332 -12.2877 -2.1852>, rot=<-0.0053 -0.0234 0.1806 -0.9833>}
 dyn_hairCurlCorr_l {parent=Dyn_hair_back_L_01, loc=<4.0507 8.9393 -1.4297>, rot=<0.0964 -0.194 -0.5243 -0.8235>}
 dyn_hairCurlCorr_r {parent=Dyn_hair_back_L_01, loc=<3.649 -12.4012 -2.9168>, rot=<-0.0928 -0.1914 0.3655 -0.9062>}
]
variables [IsFE {type=bool, default=False}]
