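# Physics asset for the back hair of the F_MED_ElegantLilyAnkle female cosmetic's face accessory, apparently for secondary hair motion against the upper body.
# A four-capsule Dyn_hair_back_L chain hangs from a small Base_hair_back_L capsule, with small swing and zero twist limit angles on its joints.
# Capsules on the spine, clavicles and upper arms form a body proxy, linked by constraints with swing and twist locked.
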PhysicsAsset /BRCosmetics/Characters/Player/Female/Medium/Bodies/F_MED_ElegantLilyAnkle/Meshes/Parts/F_MED_ElegantLilyAnkle_FaceAcc_Physics_Back
bodies {
 Base_hair_back_L [{shapeType=Capsule, length=0.505, radius=1.0734, shapeName=Base_hair_back_L_sphere, rotation=<rot 0 -90 0>}]
 Dyn_hair_back_L_01 [{
   shapeType=Capsule
   length=18.2703
   radius=4.7153
   shapeName=Dyn_hair_back_L_01_capsule
   center=<3.8171 -0.2245 1.5527>
   rotation=<rot -0.0006 0 -89.8832>
  }]
 Dyn_hair_back_L_02 [{
   shapeType=Capsule
   length=17.1823
   radius=4.5692
   shapeName=Dyn_hair_back_L_02_capsule
   center=<4.4 -0.3157 1.4849>
   rotation=<rot -0.0589 0 -89.8736>
  }]
 Dyn_hair_back_L_03 [{
   shapeType=Capsule
   length=18.2088
   radius=4.235
   shapeName=Dyn_hair_back_L_03_capsule
   center=<3.9423 -0.3006 1.1371>
   rotation=<rot -0.0003 0 -90.9903>
  }]
 Dyn_hair_back_L_04 [{
   shapeType=Capsule
   length=12.6529
   radius=5.2953
   shapeName=Dyn_hair_back_L_04_capsule
   center=<5.3842 -0.6717 1.9831>
   rotation=<rot 6.8069 0.0448 87.1785>
  }]
 clavicle_l [{
   shapeType=Capsule
   length=9.326
   radius=5.6718
   shapeName=clavicle_l_capsule
   center=<6.4229 -1.3784 -2.2501>
   rotation=<rot 86.9638 -0.1986 -5.0246>
  }]
 clavicle_r [{
   shapeType=Capsule
   length=9.326
   radius=5.6718
   shapeName=clavicle_l_capsule
   center=<-6.4725 1.6489 1.745>
   rotation=<rot 86.9638 -0.1986 174.9754>
  }]
 spine_01 [{shapeType=Capsule, length=6.4341, radius=9.8359, shapeName=spine_01_capsule, center=<0 0.3179 0>}]
 spine_02 [{shapeType=Capsule, length=4.0085, radius=9.4215, shapeName=spine_02_capsule, center=<0 1.395 0>}]
 spine_03 [{shapeType=Capsule, length=4.8535, radius=9.5365, shapeName=spine_03_capsule, center=<-2.0838 2.2176 0>}]
 spine_04 [{shapeType=Capsule, length=6.338, radius=9.8801, shapeName=spine_04_capsule, center=<0 2.8828 0>}]
 spine_05 [{shapeType=Capsule, length=11.2129, radius=8.2863, shapeName=spine_05_capsule, center=<0.0464 0.3309 0>}]
 upperarm_l [{
   shapeType=Capsule
   length=24.3425
   radius=5.6575
   shapeName=upperarm_l_capsule
   center=<12.6605 -0.1968 -0.0207>
   rotation=<rot -89.9062 179.9462 -179.9461>
  }]
 upperarm_r [{
   shapeType=Capsule
   length=24.3425
   radius=5.6575
   shapeName=upperarm_l_capsule
   center=<-12.6605 0.1968 0.0207>
   rotation=<rot -89.9062 179.9462 0.0539>
  }]
}
constraints [
 {
  bone1Name=spine_02
  bone2Name=spine_01
  swing1LimitDegrees=45
  swing1Motion=Locked
  swing2LimitDegrees=45
  swing2Motion=Locked
  twistLimitDegrees=45
  twistMotion=Locked
 }
 {
  bone1Name=spine_03
  bone2Name=spine_02
  swing1LimitDegrees=45
  swing1Motion=Locked
  swing2LimitDegrees=45
  swing2Motion=Locked
  twistLimitDegrees=45
  twistMotion=Locked
 }
 {
  bone1Name=spine_04
  bone2Name=spine_03
  swing1LimitDegrees=45
  swing1Motion=Locked
  swing2LimitDegrees=45
  swing2Motion=Locked
  twistLimitDegrees=45
  twistMotion=Locked
 }
 {
  bone1Name=spine_05
  bone2Name=spine_04
  swing1LimitDegrees=45
  swing1Motion=Locked
  swing2LimitDegrees=45
  swing2Motion=Locked
  twistLimitDegrees=45
  twistMotion=Locked
 }
 {bone1Name=Dyn_hair_back_L_01, bone2Name=Base_hair_back_L, swing1LimitDegrees=10, swing2LimitDegrees=10, twistLimitDegrees=0}
 {bone1Name=Dyn_hair_back_L_02, bone2Name=Dyn_hair_back_L_01, swing1LimitDegrees=1, swing2LimitDegrees=10, twistLimitDegrees=0}
 {bone1Name=Dyn_hair_back_L_03, bone2Name=Dyn_hair_back_L_02, swing1LimitDegrees=5, swing2LimitDegrees=0, twistLimitDegrees=0}
 {bone1Name=Dyn_hair_back_L_04, bone2Name=Dyn_hair_back_L_03, swing1LimitDegrees=5, swing2LimitDegrees=0, twistLimitDegrees=0}
 {
  bone1Name=clavicle_l
  bone2Name=spine_05
  swing1LimitDegrees=45
  swing1Motion=Locked
  swing2LimitDegrees=45
  swing2Motion=Locked
  twistLimitDegrees=45
  twistMotion=Locked
 }
 {
  bone1Name=clavicle_r
  bone2Name=spine_05
  swing1LimitDegrees=45
  swing1Motion=Locked
  swing2LimitDegrees=45
  swing2Motion=Locked
  twistLimitDegrees=45
  twistMotion=Locked
 }
 {
  bone1Name=upperarm_r
  bone2Name=clavicle_r
  swing1LimitDegrees=45
  swing1Motion=Locked
  swing2LimitDegrees=45
  swing2Motion=Locked
  twistLimitDegrees=45
  twistMotion=Locked
 }
 {
  bone1Name=upperarm_l
  bone2Name=clavicle_l
  swing1LimitDegrees=45
  swing1Motion=Locked
  swing2LimitDegrees=45
  swing2Motion=Locked
  twistLimitDegrees=45
  twistMotion=Locked
 }
]
